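# Translucent unlit Niagara mesh-particle material for the wing-top effect of the TruckScale emote.
# Radially panned flame noise distorts center-scaled UVs, with strength also driven by a particle dynamic parameter.
# Two counter-rotating soft flare samples use those UVs, are summed, then masked by Fresnel for opacity.
# Particle color feeds emissive adjustments, and a quality switch adds depth fade at Medium, High and Epic.
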Material /BRCosmetics/Animation/Game/MainPlayer/Emotes/TruckScale/FX/M_FX_EmoteTruckScale_WingTop
graph {
 properties {
  bAutomaticallySetUsageInEditor=false
  bUsedWithNiagaraMeshParticles=true
  blendMode=BLEND_Translucent
  shadingModel=MSM_Unlit
  usedShadingModels=Unlit
 }
 nodes [
  ComponentMask(None=VectorParameter() {parameterName=RotationCenter, defaultValue=<rgba 0 0 0 1>}.RGB) {g=true, r=true}
  MaterialFunctionCall(Emissive=ParticleColor().RGB) {materialFunction=&MF_FX_EmissiveAdjustments}
  n1 = TextureCoordinate()
  n2 = ComponentMask(None=VectorParameter() {parameterName=DistortionOrigin, defaultValue=<rgba 0.5 0.5 0 1>}.RGB) {g=true, r=true}
  n3 = Multiply(
   A=MaterialFunctionCall(Vector or UVs=Subtract(A=@n1, B=@n2)) {materialFunction=&VectorToRadialValue}.Radial Coordinates
   B=ComponentMask(None=VectorParameter() {parameterName=DistortionTile, defaultValue=<rgba 4 0.2 0 1>}.RGB) {g=true, r=true}
  )
  n4 = ComponentMask(None=VectorParameter() {parameterName=DistortionSpeed, defaultValue=<rgba 0 -1 0 1>}.RGB) {g=true, r=true}
  n5 = Multiply(A=TextureSample(UVs=Panner(Coordinate=@n3, Speed=@n4) {bFractionalPart=true, speedY=-1}) {samplerType=SAMPLERTYPE_LinearColor, texture=/Game/Effects/FogVolume/noise_flamy_Tile_Linear}.R) {constB=2}
  n6 = VectorParameter() {parameterName=DistortionStrength, defaultValue=<rgba 0 0 0 1>}
  n7 = Multiply(A=ComponentMask(None=@n6.RGB) {g=true, r=true}, B=@n6.B)
  n8 = DynamicParameter() {paramNames [Param1, DistortionStrength, Param3, Param4]}
  n9 = Add(
   A=MaterialFunctionCall(Texture Scale=ScalarParameter() {parameterName=Scale}) {materialFunction=&ScaleUVsByCenter}.UVs
   B=Multiply(A=Subtract(A=@n5), B=Multiply(A=@n7, B=@n8.DistortionStrength))
  )
  n10 = TextureSample(UVs=Rotator(Coordinate=@n9) {speed=3}) {texture=/Game/Effects/Fort_Effects/Textures/Flares/T_Flare_Soft}
  n11 = TextureSample(UVs=Rotator(Coordinate=@n9) {speed=-4}) {texture=&T_Flare_Soft}
  n12 = Multiply(
   A=MaterialFunctionCall(Exponent=ScalarParameter() {parameterName=FresnelExp}) {materialFunction=&MF_Fresnel_FX}.Fresnel
   B=Saturate(None=Add(A=@n10.R, B=@n11.R))
  )
  n13 = DepthFade(Opacity=@n12, FadeDistance=ScalarParameter() {parameterName=DepthFade})
  QualitySwitch(Default=@n12, High=@n13, Medium=@n13, Epic=@n13)
 ]
}
