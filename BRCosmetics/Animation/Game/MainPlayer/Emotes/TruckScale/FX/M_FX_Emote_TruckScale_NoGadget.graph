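# Unlit masked material for the TruckScale emote FX, named as its NoGadget variant and apparently meant to hide a mesh.
# It is flagged for skeletal mesh and mesh deformer use.
# The graph holds only a default-zero constant and a vector from world position to actor position, a common vertex-collapse offset.
# Output connections are not shown, so clipping through opacity mask and collapsing through World Position Offset are inferred.
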Material /BRCosmetics/Animation/Game/MainPlayer/Emotes/TruckScale/FX/M_FX_Emote_TruckScale_NoGadget
graph {
 properties {
  bAutomaticallySetUsageInEditor=false
  bUsedWithMeshDeformer=true
  bUsedWithSkeletalMesh=true
  blendMode=BLEND_Masked
  shadingModel=MSM_Unlit
  usedShadingModels=Unlit
 }
 nodes [
  Constant()
  Subtract(A=ActorPositionWS(), B=WorldPosition().XYZ)
 ]
}
