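# Unlit, two-sided translucent particle material for the thruster icons in the TruckScale emote, usable on Niagara sprites and meshes.
# Particle color feeds an emissive-adjustment function, and particle alpha multiplies a parameter-selected channel of the wing texture.
# Panned flame noise in radial coordinates around a settable origin offsets the icon UVs, with distortion strength zero by default.
# A SwizzleUV parameter blends the icon UVs toward swapped axes before that offset.
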Material /BRCosmetics/Animation/Game/MainPlayer/Emotes/TruckScale/FX/M_FX_Emote_TruckScale_ThrusterIcons
graph {
 properties {
  bAutomaticallySetUsageInEditor=false
  bUsedWithNiagaraMeshParticles=true
  bUsedWithNiagaraSprites=true
  blendMode=BLEND_Translucent
  shadingModel=MSM_Unlit
  twoSided=true
  usedShadingModels=Unlit
 }
 nodes [
  n1 = ParticleColor()
  MaterialFunctionCall(Emissive=@n1.RGB) {materialFunction=&MF_FX_EmissiveAdjustments}
  n2 = TextureCoordinate()
  n3 = ComponentMask(None=VectorParameter() {parameterName=DistortionOrigin, defaultValue=<rgba 0.5 0.5 0 1>}.RGB) {g=true, r=true}
  n4 = Multiply(
   A=MaterialFunctionCall(Vector or UVs=Subtract(A=@n2, B=@n3)) {materialFunction=&VectorToRadialValue}.Radial Coordinates
   B=ComponentMask(None=VectorParameter() {parameterName=DistortionTile, defaultValue=<rgba 4 0.2 0 1>}.RGB) {g=true, r=true}
  )
  n5 = ComponentMask(None=VectorParameter() {parameterName=DistortionSpeed, defaultValue=<rgba 0 -1 0 1>}.RGB) {g=true, r=true}
  n6 = Multiply(A=TextureSample(UVs=Panner(Coordinate=@n4, Speed=@n5) {bFractionalPart=true, speedY=-1}) {samplerType=SAMPLERTYPE_LinearColor, texture=/Game/Effects/FogVolume/noise_flamy_Tile_Linear}.R) {constB=2}
  n7 = VectorParameter() {parameterName=DistortionStrength, defaultValue=<rgba 0 0 0 1>}
  n8 = ComponentMask(None=@n7.RGB) {g=true, r=true}
  n9 = Add(
   A=LinearInterpolate(
    A=TextureCoordinate()
    B=MaterialFunctionCall() {materialFunction=&Swizzle}.YX
    Alpha=ScalarParameter() {parameterName=SwizzleUV}
   )
   B=Multiply(A=Subtract(A=@n6), B=Multiply(A=@n8, B=@n7.B))
  )
  Multiply(
   A=@n1.A
   B=DotProduct(
    A=TextureSample(UVs=@n9) {texture=&/T_FX_TruckScale_Emote_Wing}.RGB
    B=VectorParameter() {parameterName=TextureChannel, defaultValue=<rgba 1 0 0 1>}.RGB
   )
  )
 ]
}
