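# Unlit translucent material for the gadget effect in the TruckScale emote, usable on Niagara mesh particles and skeletal meshes.
# A distorted sphere mask forms a light spot, its radius scaled by a dynamic parameter and tinted by particle color.
# Panned plasma and flame noise in radial coordinates supply color detail and distortion.
# A wing texture's inverted green channel modulates brightness before the emissive adjustments function.
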Material /BRCosmetics/Animation/Game/MainPlayer/Emotes/TruckScale/FX/M_FX_Emote_Truckscale_Gadget
graph {
 properties {
  bAutomaticallySetUsageInEditor=false
  bUsedWithMeshDeformer=true
  bUsedWithNiagaraMeshParticles=true
  bUsedWithSkeletalMesh=true
  blendMode=BLEND_Translucent
  shadingModel=MSM_Unlit
  usedShadingModels=Unlit
 }
 nodes [
  n1 = TextureSample() {texture=&/T_FX_TruckScale_Emote_Wing}
  n2 = Multiply(A=Add(A=OneMinus(None=@n1.G)), B=ScalarParameter() {parameterName=EdgeBrightness})
  n3 = VectorParameter() {parameterName=NoiseColor, defaultValue=<rgba 0 0 0 1>}
  n4 = ComponentMask(None=VectorParameter() {parameterName=NoiseTile, defaultValue=<rgba 4 0.2 0 1>}.RGB) {g=true, r=true}
  n5 = VectorParameter() {parameterName=DistortionOrigin, defaultValue=<rgba 0.5 0.5 0 1>}
  n6 = MaterialFunctionCall(Vector or UVs=Subtract(A=TextureCoordinate(), B=ComponentMask(None=@n5.RGB) {g=true, r=true})) {materialFunction=&VectorToRadialValue}
  n7 = VectorParameter() {parameterName=NoiseSpeed, defaultValue=<rgba 4 0.2 0 1>}
  n8 = TextureSample(UVs=Panner(Coordinate=Multiply(A=@n4, B=@n6.Radial Coordinates), Speed=ComponentMask(None=@n7.RGB) {g=true, r=true}) {bFractionalPart=true, speedY=-1}) {samplerType=SAMPLERTYPE_LinearColor, texture=/Game/Effects/Fort_Effects/Emotes/UltraLight/Texture/T_Plasma_D_C}
  n9 = ComponentMask(None=VectorParameter() {parameterName=DistortionTile, defaultValue=<rgba 4 0.2 0 1>}.RGB) {g=true, r=true}
  n10 = VectorParameter() {parameterName=DistortionSpeed, defaultValue=<rgba 0 -1 0 1>}
  n11 = TextureSample(UVs=Panner(Coordinate=Multiply(A=@n6.Radial Coordinates, B=@n9), Speed=ComponentMask(None=@n10.RGB) {g=true, r=true}) {bFractionalPart=true, speedY=-1}) {samplerType=SAMPLERTYPE_LinearColor, texture=/Game/Effects/FogVolume/noise_flamy_Tile_Linear}
  n12 = VectorParameter() {parameterName=DistortionStrength, defaultValue=<rgba 0 0 0 1>}
  n13 = Multiply(A=Subtract(A=Multiply(A=@n11.R) {constB=2}), B=Multiply(A=ComponentMask(None=@n12.RGB) {g=true, r=true}, B=@n12.B))
  n14 = TextureCoordinate()
  n15 = VectorParameter() {parameterName=LightPosition, defaultValue=<rgba 0 0 0 1>}
  n16 = ScalarParameter() {parameterName=LightRadius}
  n17 = DynamicParameter() {paramNames [Param1,Param2,LightRadius,Param4]}
  n18 = Power(Base=SphereMask(A=Add(A=@n13, B=@n14), B=ComponentMask(None=@n15.RGB) {g=true, r=true}, Radius=Multiply(A=@n16, B=@n17.LightRadius)) {hardnessPercent=0})
  n19 = VectorParameter() {parameterName=Color, defaultValue=<rgba 0 0 0 1>}
  n20 = Multiply(A=@n18, B=ParticleColor().RGB)
  n21 = Add(A=Multiply(A=Multiply(A=@n3.RGB, B=@n8.RGB), B=@n18), B=Multiply(A=Add(A=@n19.RGB, B=@n20)))
  n22 = VectorParameter() {parameterName=OverallColor, defaultValue=<rgba 1 1 1 1>}
  MaterialFunctionCall(Emissive=Multiply(A=Multiply(A=@n2, B=@n21), B=Multiply(A=@n22.RGB, B=@n22.RGB))) {materialFunction=&MF_FX_EmissiveAdjustments}
  n23 = SphereMask(A=ComponentMask(None=TextureCoordinate() {coordinateIndex=1}) {g=true}, B=Constant() {r=0.5}) {attenuationRadius=0.5, hardnessPercent=0}
  Saturate(None=Add(A=OneMinus(None=@n23)) {constB=0.5})
  TextureSample() {samplerType=SAMPLERTYPE_LinearColor, texture=/Game/Effects/Fort_Effects/GameplayPulginsAssets/DiamondSkin/Textures/T_DiamondSkin_Flare}
  TextureSample() {texture=/Game/Effects/Fort_Effects/Textures/Flares/T_Flare_Soft}
 ]
}
