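# Translucent, unlit, two-sided Niagara sprite and mesh particle material for the wing effect in the TruckScale emote.
# Emissive color combines particle color over the inverted wing texture with a logo tint that blends toward LightColor inside a positionable sphere mask.
# Radially mapped, panning cloud noise distorts a second wing sample, scaled along V and by a dynamic parameter.
# That sample joins the undistorted mask under a Fresnel term; ShowThrusterSpikes blends identical inputs, changing nothing.
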Material /BRCosmetics/Animation/Game/MainPlayer/Emotes/TruckScale/FX/M_FX_Emote_Truckscale_Wing
graph {
 properties {
  bAutomaticallySetUsageInEditor=false
  bUsedWithNiagaraMeshParticles=true
  bUsedWithNiagaraSprites=true
  blendMode=BLEND_Translucent
  shadingModel=MSM_Unlit
  twoSided=true
  usedShadingModels=Unlit
 }
 nodes [
  Add()
  Constant3Vector() {constant=<rgba 1 1 1 0>}
  DynamicParameter() {paramNames [ShowThrusterSpikes, DistortionStrength, Param3, Param4]}
  MaterialFunctionCall(Offset Amount=ScalarParameter() {parameterName=CameraOffset}) {materialFunction=/Engine/Functions/Engine_MaterialFunctions02/WorldPositionOffset/CameraOffset}
  n1 = ParticleColor()
  n2 = TextureSample() {texture=&/T_FX_TruckScale_Emote_Wing}
  n3 = OneMinus(None=@n2.R)
  MaterialFunctionCall(Emissive=Add(A=Multiply(A=@n1.RGB, B=@n3), B=NamedRerouteUsage())) {materialFunction=&MF_FX_EmissiveAdjustments}
  n4 = Subtract(
   A=TextureCoordinate()
   B=ComponentMask(None=VectorParameter() {parameterName=DistortionOrigin, defaultValue=<rgba 0.5 0.5 0 1>}.RGB) {g=true, r=true}
  )
  n5 = VectorParameter() {parameterName=DistortionTile, defaultValue=<rgba 4 0.2 0 1>}
  n6 = Panner(
   Coordinate=Multiply(
    A=MaterialFunctionCall(Vector or UVs=@n4) {materialFunction=&VectorToRadialValue}.Radial Coordinates
    B=ComponentMask(None=@n5.RGB) {g=true, r=true}
   )
   Speed=ComponentMask(None=VectorParameter() {parameterName=DistortionSpeed, defaultValue=<rgba 0 -1 0 1>}.RGB) {g=true, r=true}
  ) {bFractionalPart=true, speedY=-1}
  n7 = Subtract(A=Multiply(A=TextureSample(UVs=@n6) {samplerType=SAMPLERTYPE_LinearGrayscale, texture=/Game/Effects/Fort_Effects/Emotes/UltraLight/Texture/T_Cloud_Greyscale}.R) {constB=2})
  n8 = VectorParameter() {parameterName=DistortionStrength, defaultValue=<rgba 0 0 0 1>}
  n9 = LinearInterpolate(
   A=ScalarParameter() {parameterName=Dist_Str_Min}
   B=ScalarParameter() {parameterName=Dist_Str_Max}
   Alpha=ComponentMask(None=TextureCoordinate()) {g=true}
  )
  n10 = DynamicParameter() {paramNames [ShowThrusterSpikes, DistortionStrength, Param3, Param4]}
  n11 = Multiply(A=Multiply(A=ComponentMask(None=@n8.RGB) {g=true, r=true}, B=Multiply(A=@n8.B, B=@n9)), B=@n10.DistortionStrength)
  n12 = TextureSample(UVs=Add(A=TextureCoordinate(), B=Multiply(A=@n7, B=@n11))) {texture=&T_FX_TruckScale_Emote_Wing}
  n13 = LinearInterpolate(A=@n12.R, B=@n12.R, Alpha=@n10.ShowThrusterSpikes)
  Multiply(
   A=MaterialFunctionCall(Exponent=ScalarParameter() {defaultValue=1, parameterName=FresnelExponent}) {materialFunction=&MF_Fresnel_FX}.Fresnel Double Sided
   B=Saturate(None=Add(A=@n13, B=@n2.R))
  )
  n14 = VectorParameter() {parameterName=LogoColor, defaultValue=<rgba 0 0 0 1>}
  n15 = VectorParameter() {parameterName=LightColor, defaultValue=<rgba 0 0 0 1>}
  n16 = VectorParameter() {parameterName=LightPos, defaultValue=<rgba 0 0 0 1>}
  n17 = Power(Base=SphereMask(A=TextureCoordinate(), B=ComponentMask(None=@n16.RGB) {g=true, r=true}, Radius=@n16.B) {hardnessPercent=0})
  NamedRerouteDeclaration(None=Multiply(A=@n2.R, B=LinearInterpolate(A=@n14.RGB, B=@n15.RGB, Alpha=@n17))) {name=LogoMain, nodeColor=<rgba 0.7344 1 0 1>}
  TextureSample() {samplerType=SAMPLERTYPE_LinearGrayscale, texture=&T_Cloud_Greyscale}
 ]
}
