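# Niagara effect for the gadget in the TruckScale emote: glowing thruster wings attached to a skeletal mesh, with rings, GPU sparks and streaks.
# Wings, streaks and sparks sample bones by particle partition, while a socket-following center particle adds flickering light.
# System timeline values drive the wing opening and fadeout, and time ring bursts at open and closedown, plus a spark burst at closedown.
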NiagaraSystem /BRCosmetics/Animation/Game/MainPlayer/Emotes/TruckScale/FX/NS_Emote_TruckScale_Gadget
dependencies {
 usedDataInterfaces [
  &NiagaraDataInterfaceCurve
  &NiagaraDataInterfaceVector2DCurve
  &NiagaraDataInterfaceColorCurve
  &NiagaraDataInterfaceSpriteRendererInfo
  &NiagaraDataInterfaceSkeletalMesh
  &NiagaraDataInterfaceSocketReader
 ]
 usedDynamicInputs [&SetBoolByFloatComparison]
 usedModules [
  &self:SystemSpawnScript.SystemScriptSource.SystemScriptGraph.NiagaraNodeAssignment_0.SetVariables_#1
  &self:SystemSpawnScript.SystemScriptSource.SystemScriptGraph.NiagaraNodeAssignment_4.SetVariables_#2
  &self:SystemSpawnScript.SystemScriptSource.SystemScriptGraph.NiagaraNodeAssignment_3.SetVariables_#3
  &self:SystemSpawnScript.SystemScriptSource.SystemScriptGraph.NiagaraNodeAssignment_1.SetVariables_#4
  &self:SystemSpawnScript.SystemScriptSource.SystemScriptGraph.NiagaraNodeAssignment_7.SetVariables_#5
  &self:SystemSpawnScript.SystemScriptSource.SystemScriptGraph.NiagaraNodeAssignment_8.SetVariables_#6
  &SystemState
  &Timeline
  &self:SystemSpawnScript.SystemScriptSource.SystemScriptGraph.NiagaraNodeAssignment_2.SetVariables_#7
  &SampleSkeletalMesh
  &self:SystemSpawnScript.SystemScriptSource.SystemScriptGraph.NiagaraNodeAssignment_5.SetVariables_#8
  &self:SystemSpawnScript.SystemScriptSource.SystemScriptGraph.NiagaraNodeAssignment_6.SetVariables_#9
 ]
}
emitters {
 Ring {
  emitterSpawnScript [/SetVariables(Emitter.GlowDepthFade=25) {isSetParameters=true}]
  emitterUpdateScript [
   &EmitterState()
   &SpawnBurst_Instantaneous(
    Spawn Count=2
    Spawn Group=1
    Spawn Time=&Add_Float(A=$System.ClosedownTime, B=0.35)
    Use Spawn Probability=false
    UseLoopCountLimit=false
   ) {module=SpawnBurst_Instantaneous002}
   &SpawnBurst_Instantaneous(Spawn Count=2, Spawn Group=1, Spawn Time=0.1, Use Spawn Probability=false, UseLoopCountLimit=false) {module=SpawnBurst_Instantaneous001}
   &SpawnBurst_Instantaneous(Spawn Count=1, Spawn Time=$System.OpenBurstTime, Use Spawn Probability=false)
  ]
  particleSpawnScript [
   /SetVariables(Particles.SpawnGroup=$Emitter.SpawnGroup) {isSetParameters=true}
   /SetVariables() {isSetParameters=true}
   /SetVariables(Particles.ShowExtras:NiagaraInt32=$Particles.SpawnGroup) {isSetParameters=true}
   &PartitionParticles(Number Of Partitions=&MergeHelper_IntSwitch(0=2, 1=2, 2=?, 3=?, 4=?, 5=?, Index=$Particles.SpawnGroup, Num=::NewEnumerator0))
   &InitializeParticle(
    Color=<rgba 1 0.9671 0.9772 1>
    Lifetime=0.25
    Position Mode=::NewEnumerator1
    Position Offset Coordinate Space=::Local
    Sprite Rotation Mode=::NewEnumerator2
    Sprite Size Mode=::NewEnumerator3
    Uniform Sprite Size=&MergeHelper_FloatSwitch(0=300, 1=275, 2=0, 3=0, 4=0, Index=$Particles.SpawnGroup, Num=::NewEnumerator3)
    UsePositionOffset=true
    Write Material Random=true
   )
   &CameraOffset(Camera Offset Amount=10)
  ]
  particleUpdateScript [
   &ParticleState()
   /SetVariables(Particles.Position=$System.WingsRoot) {enabled=false, isSetParameters=true}
   &SkeletalMeshLocation(
    Bone / Socket Index=$Particles.PartitionParticles.Partition
    Bone Sampling Mode=::NewEnumerator4
    MaskSpawnGroup=false
    Position Sampling=::NewEnumerator2
    Sampled Position Offset=<0 -30 0>
    Skeletal Mesh=$System.Skeletal Mesh
   ) {module=SkeletalMeshLocation001}
   /SetVariables(Particles.Position=&MergeHelper_VectorSwitch(
     0=Linked() {linkedVariable {name=System.WingsRoot, type=NiagaraPosition}}
     1=$Particles.SkeletalMeshLocation001.SampledPosition
     2=?
     3=?
     Index=$Particles.SpawnGroup
     Num=::NewEnumerator0
    )) {isSetParameters=true}
   &ScaleColor(
    Scale Alpha=&FloatFromCurve(FloatCurve=DataInterface() {propertyValues {Curve {keys [<at 0 0.5>,<at 1 0>]}}})
    Scale RGB=&MakeVectorFromLinearColorRGB(Color=&ColorFromCurve(ColorCurve=DataInterface() {propertyValues {
        AlphaCurve {keys [<at 0 1 RCIM_Cubic>,<at 1 0 RCIM_Cubic>]}
        BlueCurve {keys [<at 0.0117 1>, {arriveTangent=-1.5394, interpMode=RCIM_Cubic, leaveTangent=-1.5394, time=0.0431, value=0.0965}, <at 0.6613 0 RCIM_Cubic>]}
        GreenCurve {keys [<at 0.0117 1>,<at 0.0431 2.5 RCIM_Cubic>,<at 0.6613 1 RCIM_Cubic>]}
        RedCurve {keys [
          <at 0.0117 1>
          {arriveTangent=-0.2284, interpMode=RCIM_Cubic, leaveTangent=-0.2284, time=0.0431, value=0.6561}
          <at 0.6613 0.8516 RCIM_Cubic>
         ]}
       }}))
    ScaleA=true
    ScaleRGB=true
   )
   &ScaleSpriteSize(Scale Factor=&Vector2DFromFloat(Value=&FloatFromCurve(FloatCurve=DataInterface() {propertyValues {Curve {keys [{arriveTangent=2.7513, interpMode=RCIM_Cubic, leaveTangent=2.7513, tangentMode=RCTM_User, time=0, value=0}, <at 1 1>]}}})))
   /SetVariables(
    Particles.FlareColor=&MultiplyLinearColor(
     Linear Color=&ColorFromCurve(ColorCurve=DataInterface() {propertyValues {
        AlphaCurve {keys [<at 0 1 RCIM_Cubic>,<at 1 0 RCIM_Cubic>]}
        BlueCurve {keys [<at 0 0.0307 RCIM_Cubic>,<at 0.9935 0.0405>]}
        GreenCurve {keys [<at 0 3 RCIM_Cubic>,<at 0.9935 1.05>]}
        RedCurve {keys [<at 0 0.7245 RCIM_Cubic>,<at 0.9935 0.2756>]}
       }})
     Scale Factor=&MakeVector4_FromFloat(Float=1)
    )
    Particles.FlareSize:Vector2f=&Multiply_Vector2D_ByFloat(
     Float=&FloatFromCurve(FloatCurve=DataInterface() {propertyValues {Curve {keys [
          {interpMode=RCIM_Cubic, tangentMode=RCTM_User, time=0, value=0}
          {interpMode=RCIM_Cubic, tangentMode=RCTM_User, time=0.3, value=0.95}
          {tangentMode=RCTM_User, time=1, value=0}
         ]}}})
     Vector2D=$Particles.Initial.SpriteSize
    )
   ) {isSetParameters=true}
   &SkeletalMeshLocation(
    Bone / Socket Index=1
    Bone Sampling Mode=::NewEnumerator4
    MaskSpawnGroup=false
    Sampled Position Offset=<0 -30 0>
    Skeletal Mesh=$User.SkeletalMesh
   ) {enabled=false}
   &FN_Sprite_Attributes(
    Color=&MakeLinearColorFromVectorAndFloat(
     "Vector (RGB)"=&MakeVectorFromLinearColorRGB(Color=<rgba 0.2965 1 0 1>)
     "Float (Alpha)"=&Multiply_Float(
      A=&Multiply_Float(
       A=1
       B=&FloatFromCurve(FloatCurve=DataInterface() {propertyValues {Curve {keys [<at 0 0>,<at 0.124 1>,<at 0.496 1>,<at 0.62 0>]}}})
      )
      B=1
     )
    )
    Create Color Attribute=true
    Create Position Attribute=true
    Create Sprite Rotation Attribute=true
    Create SpriteSize Attribute=true
    Position:NiagaraPosition=&AddVectorToPosition(
     Position=$Particles.Position
     Vector=&MakeVector(
      X=0
      Y=0
      Z=&FloatFromCurve(FloatCurve=DataInterface() {propertyValues {Curve {keys [<at 0 -0.5>,<at 1 1>]}}}, Scale Curve=70)
     )
    )
    SpriteSize=&Multiply_Vector2D_ByFloat(
     Float=&Multiply_Float(
      A=300
      B=&FloatFromCurve(FloatCurve=DataInterface() {propertyValues {Curve {keys [
           {arriveTangent=4.1785, interpMode=RCIM_Cubic, leaveTangent=4.1785, tangentMode=RCTM_User, time=0, value=0}
           {arriveTangent=0.9995, interpMode=RCIM_Cubic, leaveTangent=0.9995, tangentMode=RCTM_User, time=0.1378, value=0.6825}
           {arriveTangent=-0.0405, interpMode=RCIM_Cubic, leaveTangent=-0.0405, tangentMode=RCTM_User, time=0.62, value=1}
          ]}}})
     )
     Vector2D=<1 1>
    )
    Write to Module Namespace=true
   )
  ]
  rendererClasses [&NiagaraSpriteRendererProperties]
  renderers [
   {rendererClass=&NiagaraSpriteRendererProperties}
   {rendererClass=&NiagaraSpriteRendererProperties, rendererIndex=1}
   {rendererClass=&NiagaraSpriteRendererProperties, rendererIndex=2}
  ]
 }
 Sparks_Wings {
  simTarget=GPUComputeSim
  emitterUpdateScript [
   &EmitterState(Life Cycle Mode=::NewEnumerator1, Loop Behavior=::NewEnumerator1, Loop Delay=0.2, UseLoopDelay=false)
   &SpawnBurst_Instantaneous(Spawn Count=50, Spawn Group=1, Spawn Time=$System.OpenBurstTime, Use Spawn Probability=false) {enabled=false}
   &SpawnBurst_Instantaneous(
    Spawn Count=80
    Spawn Time=&Add_Float(A=$System.ClosedownTime, B=0.4)
    Use Spawn Probability=false
    UseLoopCountLimit=false
   ) {module=SpawnBurst_Instantaneous001}
   &SpawnRate(
    SpawnRate=&FloatFromCurve(FloatCurve=DataInterface() {propertyValues {Curve {keys [<at 0 100>,<at 0.15 100>,<at 0.2 0>]}}}, CurveIndex=$Emitter.NormalizedLoopAge)
    Use Spawn Probability=false
   ) {enabled=false}
  ]
  particleSpawnScript [
   /Niagara/Modules/Spawn/Initialization/InitializeParticle(
    Color=<rgba 10 6.0112 3.2412 1>
    Lifetime=&UniformRangedFloat(Minimum=0.5, Maximum=1.5)
    Module.Write Color=true
    Module.Write Lifetime=true
    Module.Write Mass=false
    Module.Write Material Random=true
    Module.Write Position=false
    Module.Write Scale=false
    Module.Write SpriteRotation=true
    Module.Write SpriteSize=true
    Position=&SimulationPosition()
    Sprite Rotation=&UniformRangedFloat(Minimum=-180, Maximum=180)
    Sprite Size=&Vector2DFromFloat(Value=&UniformRangedFloat(Minimum=1.25, Maximum=2.25))
   ) {enabled=false}
   /Niagara/Modules/Spawn/Initialization/V2/InitializeParticle(
    Color=<rgba 1.9916 2 1.9826 1>
    Lifetime=&UniformRangedFloat(Minimum=0.1, Maximum=1)
    Sprite Rotation Angle=&UniformRangedFloat(Minimum=-180, Maximum=180)
    Sprite Rotation Mode=::NewEnumerator0
    Sprite Size=&Vector2DFromFloat(Value=&UniformRangedFloat(Minimum=1, Maximum=1))
    Sprite Size Mode=::NewEnumerator4
    UsePositionOffset=false
    Write Material Random=true
   ) {module=InitializeParticle001}
   &PartitionParticles()
   &SkeletalMeshLocation(Bone / Socket Index=1, Bone Sampling Mode=::NewEnumerator4, MaskSpawnGroup=false, Skeletal Mesh=$User.SkeletalMesh) {enabled=false, module=SkeletalMeshLocation001}
   &SkeletalMeshLocation(
    Bone / Socket Index=$Particles.PartitionParticles.Partition
    Bone Sampling Mode=::NewEnumerator4
    MaskSpawnGroup=false
    Skeletal Mesh=$System.Skeletal Mesh
   )
   &AddVelocity(Velocity=&MakeVector(X=0, Y=0, Z=&UniformRangedFloat(Minimum=25, Maximum=75))) {enabled=false}
   /Niagara/Modules/Spawn/Location/SphereLocation(
    MaskSpawnGroup=false
    Module.Override Randomness=false
    Offset=&MaskVectorBySpawnGroup(Masked Vector=<0 0 0>, Passthrough Vector=<0 0 0>)
    Sphere Radius=20
    Surface Only=false
   )
   &AddVelocity(
    Cone Angle=6070
    Cone Axis=&RotateVectorByQuaternion(Vector To Rotate=<0 1 -1>, Quaternion=$Particles.SkeletalMeshLocation.SampledRotation)
    Module.UseConeFalloff=false
    Module.UseVelDistribution=false
    Override Seed=false
    Velocity Mode=::NewEnumerator2
    Velocity Speed=&RandomRangeFloat(Minimum=100, Maximum=400)
   ) {module=AddVelocity002}
   &ApplyInitialForces()
   &AddVelocity() {enabled=false, module=AddVelocity001}
   &AddVelocity(
    Velocity=<0 0 -1>
    Velocity Speed Scale=&MaskFloatBySpawnGroup(Masked Float=0, Passthrough Float=&RandomRangeFloat(Minimum=100, Maximum=250))
   ) {enabled=false, module=AddVelocity003}
  ]
  particleUpdateScript [
   &ParticleState()
   /Niagara/Modules/Update/Forces/CurlNoiseForce(
    Noise Frequency=0.0125
    Noise Strength=&FloatFromCurve(FloatCurve=DataInterface() {propertyValues {Curve {keys [<at 0 1>,<at 1 0>]}}}, Scale Curve=1000)
    Pan Noise=true
    Pan Noise Field=<3 0 0>
   )
   &AccelerationForce(Acceleration=&RandomRangeVector(Minimum=<0 0 -250>, Maximum=<0 0 2>))
   &Drag(Drag=2, Use Linear Drag=true, Use Rotational Drag=false)
   &ScaleColor(
    Scale Alpha=&FloatFromCurve(FloatCurve=DataInterface() {propertyValues {Curve {keys [
         {arriveTangent=11.9433, interpMode=RCIM_Cubic, leaveTangent=11.9433, tangentMode=RCTM_User, time=0, value=0}
         <at 0.2 1 RCIM_Cubic>
         <at 1 0 RCIM_Cubic>
        ]}}})
    Scale RGB=&MakeVectorFromLinearColorRGB(Color=&ColorFromCurve(ColorCurve=DataInterface() {propertyValues {
        AlphaCurve {keys [<at 0 1 RCIM_Cubic>]}
        BlueCurve {keys [<at 0 0 RCIM_Cubic>,<at 0.5253 0.0003 RCIM_Cubic>]}
        GreenCurve {keys [<at 0 5 RCIM_Cubic>,<at 0.5253 0.3802 RCIM_Cubic>]}
        RedCurve {keys [<at 0 1.9155 RCIM_Cubic>,<at 0.5253 0.2691 RCIM_Cubic>]}
       }}))
    ScaleA=true
    ScaleRGB=true
   )
   &DynamicMaterialParameters(
    Index 0 Param 1=&FloatFromCurve(FloatCurve=DataInterface() {propertyValues {Curve {keys [<at 0 0>,<at 1 1>]}}})
    Index 0 Param 2=1
    Index 0 Param 3=1
    Index 0 Param 4=1
    Module.Param0WriteEnabled=true
    Module.Param1WriteEnabled=true
    Module.Param2WriteEnabled=true
    Module.Param3WriteEnabled=true
   )
   &ScaleSpriteSize(Scale Factor=&Vector2DFromFloat(Value=&FloatFromCurve(FloatCurve=DataInterface() {propertyValues {Curve {keys [
          {arriveTangent=9.1531, interpMode=RCIM_Cubic, leaveTangent=9.1531, tangentMode=RCTM_User, time=0, value=0.02}
          {arriveTangent=0.48, interpMode=RCIM_Cubic, leaveTangent=0.48, time=0.18, value=0.99}
          <at 1 0.5 RCIM_Cubic>
         ]}}})))
   &PointAttractionForce(
    Attraction Radius=2500
    Attraction Strength=50
    Attractor Position:NiagaraPosition=$Particles.Initial.Position
    Kill Within Radius=false
    Module.Use Falloff=true
    UseAttractorOffset=false
   )
   &InheritSourceMovement(
    Applied Position Scale=&VectorFromFloat(Value=&FloatFromCurve(FloatCurve=DataInterface() {propertyValues {Curve {keys [<at 0 1>,<at 0.6 0.825>,<at 1 0>]}}}))
    ApplyForce=false
    ApplyPosition=true
    ApplyPositionFalloff=false
   )
   &SolveForcesAndVelocity(Clamp Velocity=false, Limit Acceleration=false)
   &ScaleSpriteSizeBySpeed(Max Scale Factor=<1 4>, Velocity Threshold=100)
  ]
  rendererClasses [&NiagaraSpriteRendererProperties, &NiagaraMeshRendererProperties]
  renderers [
   {rendererClass=&NiagaraSpriteRendererProperties}
   {rendererClass=&NiagaraMeshRendererProperties, rendererIndex=1}
   {rendererClass=&NiagaraSpriteRendererProperties, rendererIndex=2}
  ]
 }
 Streaks {
  emitterSpawnScript [/SetVariables(Emitter.BurstReach=1.2, Emitter.BurstSpeed=2) {isSetParameters=true}]
  emitterUpdateScript [
   &EmitterState(
    Inactive Response=::NewEnumerator1
    Life Cycle Mode=::NewEnumerator1
    Loop Behavior=::NewEnumerator1
    Loop Delay=0.1
    Loop Duration=1
    UseLoopDelay=true
   )
   &SpawnBurst_Instantaneous(Spawn Count=4, Spawn Group=3, Use Spawn Probability=false, UseLoopCountLimit=false) {module=SpawnBurst_Instantaneous003}
   &SpawnBurst_Instantaneous(Spawn Count=5, Spawn Group=2, Use Spawn Probability=false, UseLoopCountLimit=false) {module=SpawnBurst_Instantaneous002}
   &SpawnBurst_Instantaneous(Spawn Count=4, Use Spawn Probability=false, UseLoopCountLimit=false)
   &SpawnBurst_Instantaneous(Spawn Count=4, Spawn Group=1, Use Spawn Probability=false, UseLoopCountLimit=false) {enabled=false, module=SpawnBurst_Instantaneous001}
  ]
  particleSpawnScript [
   &PartitionParticles()
   /SetVariables(Particles.SpawnGroup=$Emitter.SpawnGroup) {isSetParameters=true}
   &InitializeParticle(
    Color=$System.LogoThrusterColor
    Lifetime=&Multiply_Float(
     A=&MergeHelper_FloatSwitch(
      0=&RandomRangeFloat(Minimum=0.1, Maximum=0.2)
      1=&RandomRangeFloat(Minimum=0.25, Maximum=0.4)
      2=&RandomRangeFloat(Minimum=0.4, Maximum=0.5)
      3=&RandomRangeFloat(Minimum=0.12, Maximum=0.2)
      Index=$Emitter.SpawnGroup
      Num=::NewEnumerator2
     )
     B=$Emitter.BurstSpeed
    )
    Sprite Size Mode=::NewEnumerator3
    Sprite UV Mode=::NewEnumerator1
    Uniform Sprite Size=18.8148
    UsePositionOffset=false
    Write Material Random=true
   )
   &CameraOffset(Camera Offset Amount=10)
   &SkeletalMeshLocation(
    Bone / Socket Index=$Particles.PartitionParticles.Partition
    Bone Sampling Mode=::NewEnumerator4
    MaskSpawnGroup=false
    Skeletal Mesh=$System.Skeletal Mesh
   )
   &AddVelocity(
    Cone Angle=20
    Cone Axis=&RotateVectorByQuaternion(Vector To Rotate=<0 -1 0>, Quaternion=$StackContext.SkeletalMeshLocation.SampledRotation)
    Inner Cone Angle=15
    Module.UseConeFalloff=false
    Module.UseVelDistribution=false
    Override Seed=false
    Velocity Mode=::NewEnumerator2
    Velocity Speed=&Multiply_Float(
     A=&MergeHelper_FloatSwitch(
      0=&RandomRangeFloat(Minimum=135.1878, Maximum=250)
      1=&RandomRangeFloat(Minimum=200, Maximum=350)
      2=10
      3=&RandomRangeFloat(Minimum=5, Maximum=5)
      4=300
      Index=$Particles.SpawnGroup
      Num=::NewEnumerator3
     )
     B=$Emitter.BurstReach
    )
   ) {module=AddVelocity002}
   /SetVariables(Particles.VisibilityTag=$Emitter.SpawnGroup) {isSetParameters=true}
  ]
  particleUpdateScript [
   &ParticleState()
   &ScaleSpriteSize(
    Scale Factor=&Multiply_Vector2D(A=&RandomRangeVector2D(Minimum=<2 2>, Maximum=<3 3>), B=<1.5794 1>)
    Scale Sprite Size Mode=::NewEnumerator2
   )
   &ScaleSpriteSize(
    Non-Uniform Curve Scale=<0.8 1>
    Non-Uniform Curve Sprite Scale=DataInterface() {propertyValues {
      XCurve {keys [
        {arriveTangent=0.8609, interpMode=RCIM_Cubic, leaveTangent=0.8609, tangentMode=RCTM_User, time=0, value=0.6}
        {arriveTangent=-0.9108, interpMode=RCIM_Cubic, leaveTangent=-0.9108, tangentMode=RCTM_User, time=0.3, value=0.5}
        {arriveTangent=0.0224, interpMode=RCIM_Cubic, leaveTangent=0.0224, tangentMode=RCTM_User, time=1, value=0.5}
       ]}
      YCurve {keys [
        {arriveTangent=0.8801, interpMode=RCIM_Cubic, leaveTangent=0.8801, tangentMode=RCTM_User, time=0, value=0.3}
        {arriveTangent=0.6586, interpMode=RCIM_Cubic, leaveTangent=0.6586, tangentMode=RCTM_User, time=0.2, value=0.85}
        {arriveTangent=-0.4778, interpMode=RCIM_Cubic, leaveTangent=-0.4778, tangentMode=RCTM_User, time=1, value=2}
       ]}
     }}
    Scale Sprite Size Mode=::NewEnumerator3
   ) {module=ScaleSpriteSize001}
   &ScaleSpriteSize(Scale Sprite Size Mode=::NewEnumerator0, Uniform Scale Factor=0.9) {module=ScaleSpriteSize002}
   &Drag(Drag=2, Use Linear Drag=true, Use Rotational Drag=true)
   &SolveForcesAndVelocity(Clamp Velocity=false, Limit Acceleration=false)
   &ScaleColor(
    Scale Alpha=&Multiply_Float(
     A=&FloatFromCurve(FloatCurve=DataInterface() {propertyValues {Curve {keys [<at 0 1 RCIM_Cubic>, {arriveTangent=-2.0013, interpMode=RCIM_Cubic, leaveTangent=-2.0013, tangentMode=RCTM_User, time=1, value=0}]}}})
     B=0.5
    )
    Scale RGB=&VectorFromFloat(Value=&FloatFromCurve(
      FloatCurve=DataInterface() {propertyValues {Curve {keys [
          {arriveTangent=9.0725, interpMode=RCIM_Cubic, leaveTangent=9.0725, tangentMode=RCTM_User, time=0, value=1}
          {arriveTangent=-2.5249, interpMode=RCIM_Cubic, leaveTangent=-2.5249, tangentMode=RCTM_User, time=0.1408, value=5}
          {arriveTangent=-0.0405, interpMode=RCIM_Cubic, leaveTangent=-0.0405, tangentMode=RCTM_User, time=1, value=1}
         ]}}}
      Scale Curve=0.25
     ))
    ScaleA=true
    ScaleRGB=true
   ) {module=ScaleColor001}
   &ScaleColor(
    Scale Mode=::NewEnumerator1
    Scale RGBA=&MakeFloat4FromVectorAndFloat(
     "Vector (RGB)"=&Lerp_Vector(
      A=<1 0.4 0.4>
      B=<1 1 1>
      Alpha=&FloatFromCurve(FloatCurve=DataInterface() {propertyValues {Curve {keys [{arriveTangent=-4.0059, interpMode=RCIM_Cubic, leaveTangent=-4.0059, tangentMode=RCTM_User, time=0.5, value=1}, <at 1 0 RCIM_Cubic>]}}})
     )
     "Float (Alpha)"=1
    )
    ScaleRGBA=true
   ) {module=ScaleColor002}
   &DynamicMaterialParameters(
    Index 0 Param 1=&FloatFromCurve(FloatCurve=DataInterface() {propertyValues {Curve {keys [
         {arriveTangent=-0.1846, time=0, value=0}
         {arriveTangent=1.3385, interpMode=RCIM_Cubic, leaveTangent=1.1594, tangentMode=RCTM_Break, time=0.5, value=0}
         {arriveTangent=2, interpMode=RCIM_Cubic, leaveTangent=2, time=0.8, value=0.5}
         {arriveTangent=3.0091, interpMode=RCIM_Cubic, leaveTangent=3.0091, tangentMode=RCTM_User, time=1, value=1}
        ]}}})
    Index 0 Param 2=1
    Index 0 Param 4=&FloatFromCurve(FloatCurve=DataInterface() {propertyValues {Curve {keys [<at 0 0 RCIM_Cubic>,<at 1 1 RCIM_Cubic>]}}}, Scale Curve=0.8)
    Module.Index0Param0Write=true
    Module.Index0Param1Write=true
    Module.Index0Param2Write=true
    Module.Index0Param3Write=true
   )
   &FN_Sprite_Attributes(
    Color=&MultiplyLinearColor(
     Linear Color=$System.LogoThrusterColor
     Scale Factor=&MakeFloat4FromVectorAndFloat(
      "Vector (RGB)"=<1 1 1>
      "Float (Alpha)"=&FloatFromCurve(
       FloatCurve=DataInterface() {propertyValues {Curve {keys [
           <at 0 0 RCIM_Cubic>
           {arriveTangent=-1.2233, interpMode=RCIM_Cubic, leaveTangent=-1.2233, tangentMode=RCTM_User, time=0.1, value=1}
           <at 1 0 RCIM_Cubic>
          ]}}}
       Scale Curve=&Multiply_Float(A=1, B=&FNTimeOfDayScalar(Night Scale=0.3, Morning Scale=0.8, Evening Scale=0.8))
      )
     )
    )
    Create Color Attribute=true
    Create SpriteSize Attribute=true
    SpriteSize=&Multiply_Vector2D_ByFloat(
     Float=5
     Vector2D=&Multiply_Vector2D(
      A=&RandomRangeVector2D(Minimum=<5 20>, Maximum=<15 30>)
      B=&Vector2DFromCurve(
       Vector2Curve=DataInterface() {propertyValues {
         XCurve {keys [{arriveTangent=0.8609, time=0, value=0.6}, {arriveTangent=0.0224, leaveTangent=0.0224, time=1, value=1}]}
         YCurve {keys [{arriveTangent=0.8801, time=0, value=0.3}, {arriveTangent=-0.4778, leaveTangent=-0.4778, time=1, value=0.75}]}
        }}
       ScaleCurve=<1 1>
      )
     )
    )
    Write to Module Namespace=true
   )
   &SubUVAnimation(
    "Sprite Renderer "=DataInterface() {propertyValues {SpriteRenderer=&self:Streaks_1.NiagaraSpriteRendererProperties_2}}
    UseEndFrame=false
    UseStartFrame=false
   )
   &SubUVAnimation(
    "Sprite Renderer "=DataInterface() {propertyValues {SpriteRenderer=&self:Streaks_1.NiagaraSpriteRendererProperties_4}}
    UseEndFrame=false
    UseStartFrame=false
   ) {module=SubUVAnimation001}
   &FN_Sprite_Attributes(
    Color=&MultiplyLinearColor(Linear Color=$Particles.Color, Scale Factor=<1 1 1 0.1>)
    Create Color Attribute=true
    Create SpriteSize Attribute=true
    SpriteSize=<50 150>
    Write to Module Namespace=true
   ) {module=FN_Sprite_Attributes001}
   &FN_Sprite_Attributes(
    Color=&MultiplyLinearColor(
     Linear Color=$System.LogoThrusterColor
     Scale Factor=&MakeFloat4FromVectorAndFloat(
      "Vector (RGB)"=&VectorFromFloat(Value=&FloatFromCurve(FloatCurve=DataInterface() {propertyValues {Curve {keys [<at 0 5>,<at 1 1>]}}}))
      "Float (Alpha)"=&FloatFromCurve(FloatCurve=DataInterface() {propertyValues {Curve {keys [<at 0 1>,<at 1 0>]}}})
     )
    )
    Create Color Attribute=true
    Create SpriteSize Attribute=true
    SpriteSize=<22.1686 56.1705>
    Write to Module Namespace=true
   ) {module=FN_Sprite_Attributes002}
   /SetVariables() {isSetParameters=true}
  ]
  rendererClasses [&NiagaraSpriteRendererProperties]
  renderers [
   {rendererClass=&NiagaraSpriteRendererProperties}
   {rendererClass=&NiagaraSpriteRendererProperties, rendererIndex=1}
   {rendererClass=&NiagaraSpriteRendererProperties, rendererIndex=2}
   {rendererClass=&NiagaraSpriteRendererProperties, rendererIndex=3}
   {rendererClass=&NiagaraSpriteRendererProperties, rendererIndex=4}
  ]
 }
 WingCenter {
  emitterUpdateScript [
   &EmitterState()
   &SpawnRate(SpawnRate=20, Use Spawn Probability=false) {enabled=false}
   &SpawnBurst_Instantaneous(Use Spawn Probability=false, UseLoopCountLimit=false)
  ]
  particleSpawnScript [&InitializeParticle(
    Color=<rgba 0.4143 1 0.042 1>
    Lifetime=$System.CurrentLoopDuration
    Position Offset Coordinate Space=::World
    Sprite Size Mode=::NewEnumerator3
    Uniform Sprite Size=55.9706
    UsePositionOffset=true
    Write Material Random=true
   )]
  particleUpdateScript [
   &ParticleState()
   &SkeletalMeshLocation(
    Bone / Socket Index=2
    MaskSpawnGroup=false
    Mesh Sampling Type=::NewEnumerator5
    Skeletal Mesh=$System.Skeletal Mesh
    Socket Sampling Mode=::NewEnumerator5
   ) {enabled=false}
   &SocketLocation(
    SamplePosOffset=true
    Sampled Position Offset=<0 -50 0>
    Socket Index=2
    Socket Reader=$System.Socket Reader
    Socket Sampling Mode=::NewEnumerator3
    VelocitySpeedLimit=false
   )
   /SetVariables(Particles.LightIntensity:NiagaraFloat=&Multiply_Float(
     A=&Multiply_Float(
      A=&FloatFromCurve(
       FloatCurve=DataInterface() {propertyValues {Curve {keys [
           <at 0 0 RCIM_Cubic>
           {arriveTangent=-3.5968, interpMode=RCIM_Cubic, leaveTangent=-3.5968, tangentMode=RCTM_User, time=0.1, value=2}
           {arriveTangent=-1.1111, interpMode=RCIM_Cubic, leaveTangent=-1.1111, time=0.2, value=1}
           <at 1 1 RCIM_Cubic>
          ]}}}
       Scale Curve=50
      )
      B=&Waveform("Waveform [1]"=::NewEnumerator12, "[1] Frequency"=10, "[1] Amplitude Min/Max"=<0.85 1.1>, "[1] Scale"=5)
     )
     B=$System.Fadeout
    )) {isSetParameters=true}
   /SetVariables(Particles.ComponentLightIntensity:NiagaraFloat=&Multiply_Float(A=$Particles.LightIntensity, B=10)) {isSetParameters=true}
   &ScaleColor(
    Scale Alpha=&FloatFromCurve(FloatCurve=DataInterface() {propertyValues {Curve {keys [<at -0.0046 1.0113 RCIM_Cubic>,<at 1.0011 0.002 RCIM_Cubic>]}}})
    ScaleA=true
    ScaleRGB=false
   )
   /SetVariables(Particles.Position=$Output.SocketLocation.SampledPosition) {enabled=false, isSetParameters=true}
   &SolveForcesAndVelocity(Clamp Velocity=false, Limit Acceleration=false)
   &CameraOffset(Camera Offset Amount=200) {enabled=false}
   &FN_Light_Attributes(
    Color=$System.LogoThrusterColor
    Create Color Attribute=true
    Create Position Attribute=true
    Create Radius Attribute=true
    Enable Particle Light Enable=true
    Light Enabled=true
    Position=&AddVectorToPosition(Position=&AddVectorToPosition(Position=$Particles.Position, Vector=<0 0 0>), Vector=<0 0 0>)
    Radius=&Multiply_Float(
     A=75
     B=&Multiply_Float(
      A=&FloatFromCurve(
       FloatCurve=DataInterface() {propertyValues {Curve {keys [
           {arriveTangent=23.3484, interpMode=RCIM_Cubic, time=0, value=0}
           {arriveTangent=2.25, interpMode=RCIM_Cubic, leaveTangent=2.25, time=0.3, value=2}
           {arriveTangent=-1.25, interpMode=RCIM_Cubic, leaveTangent=-1.25, time=0.5, value=1.125}
           <at 1.1 1 RCIM_Cubic>
          ]}}}
       CurveIndex=&Multiply_Float(A=$System.Age, B=2)
      )
      B=&Waveform("Waveform [1]"=::NewEnumerator12, "[1] Frequency"=25, "[1] Amplitude Min/Max"=<0.9 1.1>)
     )
    )
    Scale Color=$Particles.LightIntensity
   )
   /SetVariables(Particles.ComponentLightRadius:NiagaraFloat=&Multiply_Float(A=$Particles.FN_Light_Attributes.Radius, B=2)) {isSetParameters=true}
   /SetVariables(Particles.SpriteSize=&Vector2DFromFloat(Value=$Particles.FN_Light_Attributes.Radius)) {isSetParameters=true}
  ]
  rendererClasses [
   &NiagaraSpriteRendererProperties
   &NiagaraComponentRendererProperties
   &NiagaraLightRendererProperties
  ]
  renderers [
   {rendererClass=&NiagaraSpriteRendererProperties}
   {rendererClass=&NiagaraComponentRendererProperties, rendererIndex=1}
   {rendererClass=&NiagaraLightRendererProperties, rendererIndex=2}
  ]
 }
 Wings {
  emitterUpdateScript [
   &EmitterState()
   &SpawnBurst_Instantaneous(Spawn Count=2, Spawn Group=1, Spawn Time=0.2, Use Spawn Probability=false, UseLoopCountLimit=false) {module=SpawnBurst_Instantaneous001}
   &SpawnBurst_Instantaneous(Spawn Count=2, Spawn Time=0.1, Use Spawn Probability=false, UseLoopCountLimit=false)
  ]
  particleSpawnScript [
   /SetVariables(Particles.VisibilityTag=$Emitter.SpawnGroup) {isSetParameters=true}
   &InitializeParticle(
    Color=$System.LogoThrusterColor
    Lifetime=$System.CurrentLoopDuration
    Sprite Size=<188.552 120>
    Sprite Size Mode=::NewEnumerator4
    UsePositionOffset=false
    Write Material Random=true
   )
   &PartitionParticles()
  ]
  particleUpdateScript [
   &SolveForcesAndVelocity(Clamp Velocity=false, Limit Acceleration=false)
   &ParticleState()
   &SkeletalMeshLocation(
    Bone / Socket Index=$Particles.PartitionParticles.Partition
    Bone Sampling Mode=::NewEnumerator4
    MaskSpawnGroup=false
    Skeletal Mesh=$System.Skeletal Mesh
   )
   /SetVariables(Particles.Scale=&Multiply_Vector(
     A=$Particles.SkeletalMeshLocation.SampledScale
     B=&MakeVector(
      X=1
      Y=&FloatFromCurve(
       FloatCurve=DataInterface() {propertyValues {Curve {keys [
           <at 0 1 RCIM_Cubic>
           {arriveTangent=0.125, interpMode=RCIM_Cubic, leaveTangent=0.125, time=0.1, value=1.5}
           {arriveTangent=-0.5556, interpMode=RCIM_Cubic, leaveTangent=-0.5556, time=0.4, value=1.05}
           <at 1 1>
          ]}}}
       CurveIndex=$System.OpeningTimeline
      )
      Z=1
     )
    )) {isSetParameters=true}
   &AlignSpriteToMeshOrientation(Mesh Orientation Relative Sprite Facing Vector:Vector3f=<0 0 1>, SetAlignment=true, SetFacing=true)
   &CameraOffset()
   &ScaleColor(
    Scale Alpha=&FloatFromCurve(
     FloatCurve=DataInterface() {propertyValues {Curve {keys [
         {arriveTangent=4.2291, interpMode=RCIM_Cubic, leaveTangent=4.2291, tangentMode=RCTM_User, time=0, value=0.375}
         {arriveTangent=0.0224, interpMode=RCIM_Cubic, leaveTangent=0.0224, tangentMode=RCTM_User, time=1, value=1}
        ]}}}
     Scale Curve=$System.Fadeout
    )
    Scale RGB=&VectorFromFloat(Value=&Waveform("Waveform [1]"=::NewEnumerator12, "[1] Frequency"=22, "[1] Amplitude Min/Max"=<0.7 1.5>, "[1] Scale"=2))
    ScaleA=true
    ScaleRGB=true
   )
   /SetVariables(Particles.SpriteSize=&Multiply_Vector2D_ByFloat(
     Float=&MakeFloatFromVector(Vector=$Particles.Scale, Channel=::NewEnumerator0)
     Vector2D=$Particles.Initial.SpriteSize
    )) {isSetParameters=true}
   /SetVariables(Particles.SpriteSize=&Lerp_Vector2(
     A=&Multiply_Vector2D(
      A=$StackContext.SpriteSize
      B=&Vector2DFromCurve(
       Vector2Curve=DataInterface() {propertyValues {
         XCurve {keys [<at 0 0.2 RCIM_Cubic>, {arriveTangent=0.4039, interpMode=RCIM_Cubic, leaveTangent=0.4039, tangentMode=RCTM_User, time=1, value=0.75}]}
         YCurve {keys [
           {arriveTangent=-1.1667, interpMode=RCIM_Cubic, time=0, value=0.2}
           {arriveTangent=0.4039, interpMode=RCIM_Cubic, leaveTangent=0.4039, tangentMode=RCTM_User, time=1, value=0.75}
          ]}
        }}
       CurveIndex=&Multiply_Float(A=$Particles.NormalizedAge, B=20)
       ScaleCurve=<1 2>
      )
     )
     B=$StackContext.SpriteSize
     Alpha=&FloatFromCurve(
      FloatCurve=DataInterface() {propertyValues {Curve {keys [
          {arriveTangent=4.1785, interpMode=RCIM_Cubic, leaveTangent=4.1785, tangentMode=RCTM_User, time=0, value=0}
          {arriveTangent=1.1179, interpMode=RCIM_Cubic, leaveTangent=1.1179, tangentMode=RCTM_User, time=0.4, value=0.75}
          {arriveTangent=0.0224, interpMode=RCIM_Cubic, leaveTangent=0.0224, tangentMode=RCTM_User, time=1, value=1}
         ]}}}
      Scale Curve=1.5
     )
    )) {isSetParameters=true}
   &FN_Sprite_Attributes(
    Color=&MultiplyLinearColor(
     Linear Color=&MultiplyLinearColor(
      Linear Color=$System.LogoThrusterColor
      Scale Factor=&MakeVector4(
       X=1
       Y=1
       Z=1
       W=&Waveform(
        Global Input Position=&RandomRangeFloat(Minimum=0, Maximum=1)
        "Waveform [1]"=::NewEnumerator12
        "[1] Frequency"=9
        "[1] Amplitude Min/Max"=<0 1>
        "[1] Scale"=0.05
       )
      )
     )
     Scale Factor=&MakeVector4(
      X=1
      Y=1
      Z=1
      W=&FloatFromCurve(
       FloatCurve=DataInterface() {propertyValues {Curve {keys [<at 0 0>,<at 0.4 1>]}}}
       CurveIndex=$System.OpeningTimeline
       Scale Curve=$System.Fadeout
      )
     )
    )
    Create Camera Offset Attribute=true
    Create Color Attribute=true
    Create Position Attribute=true
    Create SpriteSize Attribute=true
    Position:NiagaraPosition=&AddVectorToPosition(
     Position=$Particles.Position
     Vector=&RotateVectorByQuaternion(Vector To Rotate=<0 -50 0>, Quaternion=$StackContext.SkeletalMeshLocation.SampledRotation)
    )
    SpriteSize=&Multiply_Vector2D_ByFloat(
     Float=&MakeFloatFromVector(Vector=$Particles.Scale, Channel=::NewEnumerator0)
     Vector2D=&Vector2DFromFloat(Value=300)
    )
    Write to Module Namespace=true
   )
   &DynamicMaterialParameters(
    Index 0 Param 1=&ClampFloat(Float=&Add_Float(
      A=&FloatFromCurve(FloatCurve=DataInterface() {propertyValues {Curve {keys [<at 0 0>,<at 0.6 2>,<at 1 1>]}}}, CurveIndex=$System.OpeningTimeline)
      B=&FloatFromCurve(FloatCurve=DataInterface() {propertyValues {Curve {keys [<at 0 1>,<at 0.3 0.3>,<at 1 0>]}}}, Scale Curve=0.5)
     ))
    Index 0 Param 2=&Add_Float(
     A=&FloatFromCurve(
      FloatCurve=DataInterface() {propertyValues {Curve {keys [<at 0 0.5 RCIM_Cubic>, {arriveTangent=1, interpMode=RCIM_Cubic, leaveTangent=1, time=0.2, value=4}, <at 0.5 1 RCIM_Cubic>]}}}
      CurveIndex=$System.OpeningTimeline
     )
     B=&FloatFromCurve(FloatCurve=DataInterface() {propertyValues {Curve {keys [<at 0 1>,<at 0.1 0.15>,<at 1 0>]}}})
    )
    Index 0 Param 3=&Waveform("Waveform [1]"=::NewEnumerator12, "[1] Frequency"=12, "[1] Amplitude Min/Max"=<0.8 1.2>)
    Module.Index0Float0Write=true
    Module.Index0Float1Write=true
    Module.Index0Float2Write=true
    Module.Index0Float3Write=true
   )
   &FN_Mesh_Attributes(Color=<rgba 0.9441 1 0 1>, Create Color Attribute=true, Write to Module Namespace=true)
   /SetVariables(Particles.LightRaysColor:LinearColor=&MultiplyLinearColor(
     Linear Color=$Particles.Color
     Scale Factor=&MakeFloat4FromVectorAndFloat(
      "Vector (RGB)"=&VectorFromFloat(Value=&FloatFromCurve(FloatCurve=DataInterface() {propertyValues {Curve {keys [
            {arriveTangent=-2.5249, interpMode=RCIM_Cubic, leaveTangent=-2.5249, tangentMode=RCTM_User, time=0, value=2}
            {arriveTangent=-1, interpMode=RCIM_Cubic, leaveTangent=-1, time=0.2, value=1}
            {arriveTangent=-0.0405, interpMode=RCIM_Cubic, leaveTangent=-0.0405, tangentMode=RCTM_User, time=1, value=1}
           ]}}}))
      "Float (Alpha)"=&FloatFromCurve(
       FloatCurve=DataInterface() {propertyValues {Curve {keys [
           {arriveTangent=9.0725, interpMode=RCIM_Cubic, leaveTangent=9.0725, tangentMode=RCTM_User, time=0, value=0}
           {arriveTangent=-2.5249, interpMode=RCIM_Cubic, leaveTangent=-2.5249, tangentMode=RCTM_User, time=0.1408, value=0.9847}
           {arriveTangent=-0.0405, interpMode=RCIM_Cubic, leaveTangent=-0.0405, tangentMode=RCTM_User, time=1, value=0}
          ]}}}
       Scale Curve=0.66
      )
     )
    )) {isSetParameters=true}
  ]
  rendererClasses [&NiagaraSpriteRendererProperties, &NiagaraMeshRendererProperties]
  renderers [
   {rendererClass=&NiagaraSpriteRendererProperties}
   {rendererClass=&NiagaraMeshRendererProperties, rendererIndex=1}
   {rendererClass=&NiagaraSpriteRendererProperties, rendererIndex=2}
   {rendererClass=&NiagaraMeshRendererProperties, rendererIndex=3}
   {rendererClass=&NiagaraMeshRendererProperties, rendererIndex=4}
   {rendererClass=&NiagaraMeshRendererProperties, rendererIndex=5}
   {rendererClass=&NiagaraMeshRendererProperties, rendererIndex=6}
  ]
 }
}
summary {
 systemName=NS_Emote_TruckScale_Gadget
 emitters [
  {emitterName=Wings, rendererClasses [&NiagaraSpriteRendererProperties, &NiagaraMeshRendererProperties]}
  {
   emitterName=WingCenter
   rendererClasses [
    &NiagaraSpriteRendererProperties
    &NiagaraComponentRendererProperties
    &NiagaraLightRendererProperties
   ]
  }
  {emitterName=Ring, rendererClasses [&NiagaraSpriteRendererProperties]}
  {
   emitterName=Sparks_Wings
   simTarget=GPUComputeSim
   rendererClasses [&NiagaraSpriteRendererProperties, &NiagaraMeshRendererProperties]
  }
  {emitterName=Streaks, rendererClasses [&NiagaraSpriteRendererProperties]}
 ]
 userVariables []
}
