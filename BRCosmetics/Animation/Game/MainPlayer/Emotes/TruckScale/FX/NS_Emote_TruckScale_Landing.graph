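# Niagara system for the landing effect of the TruckScale emote, bursting dust puffs and sparks at the player's left and right foot grounded positions.
# Both emitters burst twice, using spawn group to pick the foot and a system-level thruster delay to time the second burst.
# Dust sprites spread horizontally from a ring near owner height, with extra front-end-only acceleration.
# GPU sparks fly outward with heavy drag, scaling by speed and fading out.
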NiagaraSystem /BRCosmetics/Animation/Game/MainPlayer/Emotes/TruckScale/FX/NS_Emote_TruckScale_Landing
dependencies {
 usedDataInterfaces [
  &NiagaraDataInterfaceCurve
  &NiagaraDataInterfaceCamera
  &NiagaraDataInterfaceColorCurve
  &NiagaraDataInterfaceVector2DCurve
 ]
 usedDynamicInputs [&Vector2DFromFloat]
 usedModules [
  &self:SystemSpawnScript.SystemScriptSource.SystemScriptGraph.NiagaraNodeAssignment_0.SetVariables_#1
  &self:SystemSpawnScript.SystemScriptSource.SystemScriptGraph.NiagaraNodeAssignment_6.SetVariables_#2
  &self:SystemSpawnScript.SystemScriptSource.SystemScriptGraph.NiagaraNodeAssignment_7.SetVariables_#3
  &SystemState
  &SampleSkeletalMesh
  &self:SystemSpawnScript.SystemScriptSource.SystemScriptGraph.NiagaraNodeAssignment_4.SetVariables_#4
  &self:SystemSpawnScript.SystemScriptSource.SystemScriptGraph.NiagaraNodeAssignment_5.SetVariables_#5
  &self:SystemSpawnScript.SystemScriptSource.SystemScriptGraph.NiagaraNodeAssignment_8.SetVariables_#6
 ]
}
emitters {
 DustPuffFeet {
  emitterUpdateScript [
   &EmitterState()
   /Game/Animation/Game/MainPlayer/Emotes/SeleneCobra/FX/SpawnFootDust(
    Height Threshold For Spawning:NiagaraFloat=$User.HeightThreshold
    Minimum Interval Between Spawn Events=0.2
    Right Foot Bone Index=1
    Skeletal Mesh:NiagaraDataInterfaceSkeletalMesh=$System.Skeletal Mesh
    Spawn Count=11
   ) {enabled=false}
   &SpawnBurst_Instantaneous(Spawn Count=5, Use Spawn Probability=false, UseLoopCountLimit=false)
   &SpawnBurst_Instantaneous(Spawn Count=5, Spawn Group=1, Spawn Time=$System.SecondThrusterDelay, Use Spawn Probability=false, UseLoopCountLimit=false) {module=SpawnBurst_Instantaneous001}
  ]
  particleSpawnScript [
   /SetVariables(Particles.SpawnGroup=$Emitter.SpawnGroup) {isSetParameters=true}
   &InitializeParticle(
    Color Channel Mode:ENiagaraChannelCorrelation=::NewEnumerator0
    Color Maximum=<rgba 0.35 0.3477 0.336 0.25>
    Color Minimum=<rgba 0.35 0.3404 0.3251 0.2>
    Color Mode=::NewEnumerator2
    Lifetime Max=1
    Lifetime Min=0.6
    Lifetime Mode=::NewEnumerator1
    Position=&MergeHelper_VectorSwitch(0=$System.LeftFootGroundedPos, 1=$System.RightFootGroundedPos, 2=?, 3=?, Index=$Particles.SpawnGroup, Num=::NewEnumerator0)
    Position Mode=::NewEnumerator1
    Position Offset Coordinate Space=::Local
    Sprite Rotation Mode=::NewEnumerator2
    Sprite Size Mode=::NewEnumerator1
    Uniform Sprite Size Max=44
    Uniform Sprite Size Min=33
    UsePositionOffset=true
    Write Material Random=true
   )
   &ShapeLocation(
    MaskSpawnGroup=false
    Override Seed=false
    Ring / Disc Distribution Mode:ENiagaraSphereDistributionMode=::NewEnumerator1
    Ring Radius=11
    Shape Origin=&MakeVector(
     X=&MakeFloatFromVector(Vector=Linked() {linkedVariable {name=StackContext.Position, type=NiagaraPosition}}, Channel=::NewEnumerator0)
     Y=&MakeFloatFromVector(Vector=Linked() {linkedVariable {name=StackContext.Position, type=NiagaraPosition}}, Channel=::NewEnumerator1)
     Z=&Add_Float(
      A=5
      B=&MakeFloatFromVector(Vector=Linked() {linkedVariable {name=Engine.Owner.Position, type=NiagaraPosition}}, Channel=::NewEnumerator2)
     )
    )
    Shape Primitive=::NewEnumerator4
    U Position=&ReturnNormalizedExecIndex()
    UseOwnerScale=false
   )
   &AddVelocity(Module.Use Falloff Distance=false, Override Seed=false, Velocity Mode=::NewEnumerator1, Velocity Speed=125)
   /SetVariables(Particles.Velocity=&Multiply_Vector(A=$Particles.Velocity, B=<1 1 0>)) {isSetParameters=true}
  ]
  particleUpdateScript [
   &CameraOffset(Camera Offset Amount=&FloatFromCurve(
     FloatCurve=DataInterface() {propertyValues {Curve {keys [{tangentMode=RCTM_User, time=0, value=0}, {tangentMode=RCTM_User, time=1, value=1}]}}}
     Scale Curve=44
    ))
   &Drag(Drag=2, Rotational Drag=0, Use Linear Drag=true, Use Rotational Drag=false)
   &GravityForce(Gravity=<0 0 33>)
   &ParticleState()
   &AccelerationForce(
    Acceleration=&FN_FrontEndVector3Modifier(Gameplay=<0 0 0>, Front End=<0 -333 0>)
    Coordinate Space=::Local
   )
   &InheritSourceMovement(
    Applied Position Scale=&VectorFromFloat(Value=&FloatFromCurve(FloatCurve=DataInterface() {propertyValues {Curve {keys [
          {tangentMode=RCTM_User, time=0, value=1}
          {arriveTangent=-0.0405, interpMode=RCIM_Cubic, leaveTangent=-0.0405, tangentMode=RCTM_User, time=0.7009, value=1}
          {arriveTangent=-8.0089, interpMode=RCIM_Cubic, leaveTangent=-8.0089, tangentMode=RCTM_User, time=1, value=0}
         ]}}}))
    ApplyForce=false
    ApplyPosition=true
    ApplyPositionFalloff=false
   )
   &SolveForcesAndVelocity(Clamp Velocity=false, Limit Acceleration=false)
   &ScaleSpriteSize(Uniform Curve Sprite Scale=DataInterface() {propertyValues {Curve {keys [
        {arriveTangent=4.1785, interpMode=RCIM_Cubic, leaveTangent=4.1785, tangentMode=RCTM_User, time=0, value=0}
        {arriveTangent=1.1179, interpMode=RCIM_Cubic, leaveTangent=1.1179, tangentMode=RCTM_User, time=0.2042, value=0.6228}
        {arriveTangent=0.0224, interpMode=RCIM_Cubic, leaveTangent=0.0224, tangentMode=RCTM_User, time=1, value=1}
       ]}}})
   &SpriteRotationRate(Rotation Rate=&Multiply_Float(
     A=&Lerp_Float(A=-555, B=-55, Alpha=$Particles.NormalizedAge)
     B=&DotProduct(
      Vector A=&ReturnCameraVector(Camera Vectors=::NewEnumerator3, Coordinate Space=::World)
      Vector B=$Particles.Initial.ShapeLocation.ShapeBitangent
     )
    ))
   &DynamicMaterialParameters(
    Index 0 Param 1=&Add_Float(
     A=&RandomRangeFloat(Minimum=0, Maximum=1)
     B=&FloatFromCurve(
      FloatCurve=DataInterface() {propertyValues {Curve {keys [
          <at 0 0 RCIM_Cubic>
          {arriveTangent=1.5947, interpMode=RCIM_Cubic, leaveTangent=1.5947, tangentMode=RCTM_User, time=0.2557, value=0.5257}
          <at 1 1 RCIM_Cubic>
         ]}}}
      CurveIndex=$Particles.Age
      Scale Curve=0.25
     )
    )
    Index 0 Param 3=1
    Index 0 Param 4=1
    Index 2 Param 3=2
    Index 3 Param 1=1
    Index 3 Param 2=70
    Module.Param0WriteEnabled=true
    Module.Param0WriteEnabled1=true
    Module.Param0WriteEnabled2=true
    Module.Param0WriteEnabled3=true
    Module.Param1WriteEnabled=true
    Module.Param1WriteEnabled1=true
    Module.Param1WriteEnabled2=true
    Module.Param1WriteEnabled3=true
    Module.Param2WriteEnabled=true
    Module.Param2WriteEnabled1=true
    Module.Param2WriteEnabled2=true
    Module.Param2WriteEnabled3=true
    Module.Param3WriteEnabled=true
    Module.Param3WriteEnabled1=true
    Module.Param3WriteEnabled2=true
    Module.Param3WriteEnabled3=true
    Write Parameter Index 1=true
    Write Parameter Index 2=true
    Write Parameter Index 3=true
   ) {module=DynamicMaterialParameters001}
   &CameraOffset(Camera Offset Amount=5) {module=CameraOffset001}
   &ScaleColor(
    Scale Alpha=&Multiply_Float(
     A=&FloatFromCurve(
      FloatCurve=DataInterface() {propertyValues {Curve {keys [{interpMode=RCIM_Cubic, tangentMode=RCTM_User, time=0, value=0}, <at 0.25 1 RCIM_Cubic>, <at 0.9 0>]}}}
      Scale Curve=&FN_FrontEndFloatModifier(Gameplay=1, Front End=1)
     )
     B=1
    )
    Scale RGB=<5 5 5>
    ScaleA=true
    ScaleRGB=true
   )
  ]
  rendererClasses [&NiagaraSpriteRendererProperties]
  renderers [
   {rendererClass=&NiagaraSpriteRendererProperties}
   {rendererClass=&NiagaraSpriteRendererProperties, rendererIndex=1}
  ]
 }
 Sparks {
  simTarget=GPUComputeSim
  emitterUpdateScript [
   &EmitterState()
   &SpawnBurst_Instantaneous(Spawn Count=60, Use Spawn Probability=false, UseLoopCountLimit=false)
   &SpawnBurst_Instantaneous(Spawn Count=60, Spawn Group=1, Spawn Time=$System.SecondThrusterDelay, Use Spawn Probability=false, UseLoopCountLimit=false) {module=SpawnBurst_Instantaneous001}
   &SpawnRate(SpawnRate=350, Use Spawn Probability=false) {enabled=false, module=SpawnRate001}
   &SpawnRate(SpawnRate=200, Use Spawn Probability=false) {enabled=false}
  ]
  particleSpawnScript [
   &PartitionParticles()
   /SetVariables(Particles.SourceSelect:NiagaraInt32=$Emitter.SpawnGroup) {isSetParameters=true}
   /Niagara/Modules/Spawn/Initialization/InitializeParticle(
    Color=<rgba 1.6287 10 0.3336 1>
    Lifetime=&RandomRangeFloat(Minimum=0.1, Maximum=0.6)
    Module.Write Color=true
    Module.Write Lifetime=true
    Module.Write Mass=true
    Module.Write Material Random=true
    Module.Write Position=true
    Module.Write Scale=false
    Module.Write SpriteRotation=false
    Module.Write SpriteSize=true
    Position=&MergeHelper_VectorSwitch(0=$System.LeftFootGroundedPos, 1=$System.RightFootGroundedPos, 2=?, 3=?, Index=$Particles.SourceSelect, Num=::NewEnumerator0)
    Sprite Size=&UniformRangedVector2D(Minimum=<1 2>, Maximum=<0.5 4>)
   ) {enabled=false}
   /Niagara/Modules/Spawn/Initialization/V2/InitializeParticle(
    Color=<rgba 1.6287 10 0.3336 1>
    Lifetime=&RandomRangeFloat(Minimum=0.1, Maximum=0.6)
    Position=&MergeHelper_VectorSwitch(0=$System.LeftFootGroundedPos, 1=$System.RightFootGroundedPos, 2=?, 3=?, Index=$Particles.SourceSelect, Num=::NewEnumerator0)
    Position Mode=::NewEnumerator1
    Sprite Size=&UniformRangedVector2D(Minimum=<1 2>, Maximum=<0.5 4>)
    Sprite Size Mode=::NewEnumerator4
    UsePositionOffset=false
    Write Material Random=true
   ) {module=InitializeParticle001}
   &ShapeLocation(MaskSpawnGroup=false, Override Seed=false, Sphere Radius=5, UseOwnerScale=true)
   &FN_SetSpawnGroup(Number of Spawn Groups=5) {enabled=false}
   &SkeletalMeshLocation(
    Apply Sampled Orientation to Mesh Particles=true
    Bone / Socket Index=&SpawnGroup()
    Bone Sampling Mode=::NewEnumerator4
    Inherit Velocity=false
    Interpolate Mesh Sampling=false
    MaskSpawnGroup=false
    Sampled Position Offset=<46 0 0>
    Sampled Position Offset Coordinate Space:ENiagaraMeshTransforms=::NewEnumerator3
    Skeletal Mesh=$User.Skeletal Mesh
    Use Sampled Mesh Attribute as Ribbon ID=false
   ) {enabled=false}
   &AddVelocityInCone(
    Module.Use Velocity Falloff On Cone Axis=false
    Velocity Strength=&RandomRangeFloat(Minimum=500, Maximum=750)
   ) {enabled=false}
   &AddVelocity(
    Module.Use Falloff Distance=false
    Origin Offset=<0 0 -10>
    Override Seed=false
    Velocity Mode=::NewEnumerator1
    Velocity Speed=&RandomRangeFloat(Minimum=100, Maximum=750)
   )
  ]
  particleUpdateScript [
   &Drag(Drag=8, Use Linear Drag=true, Use Rotational Drag=true)
   &SolveForcesAndVelocity(Clamp Velocity=false, Limit Acceleration=false)
   &ParticleState()
   &ScaleSpriteSizeBySpeed(Max Scale Factor=<0.5 2>, Velocity Threshold=1250)
   &ScaleColor(
    Scale Alpha=&FloatFromCurve(
     FloatCurve=DataInterface() {propertyValues {Curve {keys [<at 0 1>,<at 1 0>]}}}
     CurveIndex=&Multiply_Float(A=$Particles.NormalizedAge, B=2)
    )
    Scale RGB=&MakeVectorFromLinearColorRGB(Color=&ColorFromCurve(
      ColorCurve=DataInterface() {propertyValues {
        AlphaCurve {keys []}
        BlueCurve {keys [<at 0 0.25 RCIM_Cubic>,<at 0.5 0.5>,<at 1 0.875 RCIM_Cubic>]}
        GreenCurve {keys [<at 0 0.525 RCIM_Cubic>,<at 0.5 0.6833>,<at 1 0.9208 RCIM_Cubic>]}
        RedCurve {keys [<at 0 1 RCIM_Cubic>,<at 0.5 1>,<at 1 1 RCIM_Cubic>]}
       }}
      CurveIndex=&Multiply_Float(A=$Particles.NormalizedAge, B=2)
     ))
    ScaleA=true
    ScaleRGB=true
   )
   &ScaleSpriteSize(Scale Factor=&Vector2DFromCurve(
     Vector2Curve=DataInterface() {propertyValues {
       XCurve {keys [<at 0 1>,<at 1 0>]}
       YCurve {keys [{arriveTangent=18.2053, interpMode=RCIM_Cubic, leaveTangent=18.2053, tangentMode=RCTM_User, time=0, value=1}, <at 1 10>]}
      }}
     CurveIndex=&Multiply_Float(A=$Particles.NormalizedAge, B=2)
     ScaleCurve=<1 1>
    )) {enabled=false}
  ]
  rendererClasses [&NiagaraSpriteRendererProperties]
  renderers [{rendererClass=&NiagaraSpriteRendererProperties}]
 }
}
summary {
 systemName=NS_Emote_TruckScale_Landing
 emitters [
  {emitterName=DustPuffFeet, rendererClasses [&NiagaraSpriteRendererProperties]}
  {emitterName=Sparks, simTarget=GPUComputeSim, rendererClasses [&NiagaraSpriteRendererProperties]}
 ]
 userVariables [
  User.bIsFrontEnd {type=NiagaraBool}
  User.bIsFrontEndPreview {type=NiagaraBool}
  User.ForwardDot {type=NiagaraFloat, defaultValue=0, description=""}
  User.HeightThreshold {type=NiagaraFloat, defaultValue=0, description=""}
  User.Skeletal Mesh {
   type=NiagaraDataInterfaceSkeletalMesh
   defaultValue {
    struct=NiagaraExt_VariableValue_DataInterface
    value {dataInterface=&self:NiagaraDataInterfaceSkeletalMesh_1}
   }
  }
 ]
}
