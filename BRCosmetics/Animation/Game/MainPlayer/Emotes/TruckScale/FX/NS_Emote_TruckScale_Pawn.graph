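# Niagara system for the pawn's TruckScale emote, adding staged rocket thruster effects to the feet and calves.
# System delay variables sequence launch flames, calf flames and ignition bursts, then main foot thrusters with smoke.
# Emitters follow sampled skeletal bones or system foot positions, with a pulsing rim-glow light between the feet.
# Spawn groups, partitioned particles and switch helpers merge left/right variants into shared emitters, with waveforms driving flicker.
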NiagaraSystem /BRCosmetics/Animation/Game/MainPlayer/Emotes/TruckScale/FX/NS_Emote_TruckScale_Pawn
dependencies {
 usedDataInterfaces [
  &NiagaraDataInterfaceCurve
  &NiagaraDataInterfaceColorCurve
  &NiagaraDataInterfaceVector2DCurve
  &NiagaraDataInterfaceVectorCurve
  /Script/Niagara.NiagaraDataInterfaceVectorField
  &NiagaraDataInterfaceArrayFloat3
 ]
 usedDynamicInputs []
 usedModules [
  &self:SystemSpawnScript.SystemScriptSource.SystemScriptGraph.NiagaraNodeAssignment_0.SetVariables_#1
  &self:SystemSpawnScript.SystemScriptSource.SystemScriptGraph.NiagaraNodeAssignment_1.SetVariables_#2
  &self:SystemSpawnScript.SystemScriptSource.SystemScriptGraph.NiagaraNodeAssignment_4.SetVariables_#3
  &self:SystemSpawnScript.SystemScriptSource.SystemScriptGraph.NiagaraNodeAssignment_2.SetVariables_#4
  &self:SystemSpawnScript.SystemScriptSource.SystemScriptGraph.NiagaraNodeAssignment_3.SetVariables_#5
  &SystemState
  &SampleSkeletalMesh
  &self:SystemSpawnScript.SystemScriptSource.SystemScriptGraph.NiagaraNodeAssignment_5.SetVariables_#6
 ]
}
emitters {
 Animorphic_-_HotCoreStart_-_FlameOuterStart {
  emitterUpdateScript [
   &EmitterState()
   &SpawnRate(SpawnGroup=1, SpawnRate=10, Use Spawn Probability=false) {enabled=false}
   &SpawnBurst_Instantaneous(Spawn Count=2, Spawn Group=1, Spawn Time=$System.SecondThrusterDelay, Use Spawn Probability=false, UseLoopCountLimit=false) {module=SpawnBurst_Instantaneous001}
   &SpawnBurst_Instantaneous(Spawn Count=2, Spawn Group=2, Spawn Time=$System.MainThrusterStart, Use Spawn Probability=false)
  ]
  particleSpawnScript [
   &FN_SetSpawnGroup_v2(Number of Spawn Groups=4) {enabled=false}
   /SetVariables(Particles.SpawnGroup=$Emitter.SpawnGroup) {isSetParameters=true}
   &PartitionParticles()
   /Niagara/Modules/Spawn/Initialization/InitializeParticle(
    Color=<rgba 0.3131 1 0.3423 0.1>
    Lifetime=0.2
    Mesh Scale=<0.333 1 1>
    Module.Write Color=true
    Module.Write Lifetime=true
    Module.Write Mass=true
    Module.Write Material Random=false
    Module.Write Position=false
    Module.Write Scale=true
    Module.Write SpriteRotation=false
    Module.Write SpriteSize=true
    Position=&SimulationPosition()
    Sprite Size=&UniformRangedVector2D(Minimum=<100 4>, Maximum=<50 1>)
   ) {enabled=false}
   /Niagara/Modules/Spawn/Initialization/V2/InitializeParticle(
    Color=<rgba 0.3131 1 0.3423 0.1>
    Lifetime=0.2
    Mesh Scale=<0.333 1 1>
    Mesh Scale Mode=::NewEnumerator4
    Sprite Size=&UniformRangedVector2D(Minimum=<100 4>, Maximum=<50 1>)
    Sprite Size Mode=::NewEnumerator4
    UsePositionOffset=false
    Write Material Random=true
   ) {module=InitializeParticle001}
   /SetVariables(
    Particles.CameraOffset_2=10
    Particles.Color_2:LinearColor=<rgba 0.2429 1 0.0941 0.3>
    Particles.DynamicMaterialParameter_2:Vector4f=<1 1 1 1>
    Particles.SpriteSize_2:Vector2f=<100 100>
   ) {isSetParameters=true}
   /SetVariables(
    Particles.Color_3:LinearColor=<rgba 0.3318 5 0.0437 1>
    Particles.DynamicMaterialParameter_3:Vector4f=&MakeVector4(X=0.5, Y=1, Z=&UniformRangedFloat(Minimum=0.25, Maximum=0.35), W=1)
    Particles.SpriteSize_3:Vector2f=&Vector2DFromFloat(Value=&UniformRangedFloat(Minimum=50, Maximum=75))
   ) {isSetParameters=true}
   /SetVariables(
    Particles.Color_4:LinearColor=<rgba 0.14 0.3 0.0259 0.1>
    Particles.DynamicMaterialParameter_4:Vector4f=&MakeVector4(X=1, Y=5, Z=20, W=1)
   ) {isSetParameters=true}
   &CameraOffset(Camera Offset Amount=2)
   &DynamicMaterialParameters(
    Index 0 Param 1=5
    Module.Index0Float0Write=true
    Module.Index0Float1Write=true
    Module.Index0Float2Write=true
    Module.Index0Float3Write=true
   )
   /Niagara/Modules/Spawn/Orientation/V2/InitialMeshOrientation(Orientation Vector=<0 0 -1>, Use Orientation Vector=true, Use Rotation Vector=false) {enabled=false}
   /Niagara/Modules/Spawn/Orientation/V3/InitialMeshOrientation(Mesh Orientation Mode=::NewEnumerator2, Orientation Vector=<0 0 -1>, Use Rotation Vector=false) {module=InitialMeshOrientation001}
  ]
  particleUpdateScript [
   &ParticleState()
   &SkeletalMeshLocation(
    Bone / Socket Index=&MergeHelper_IntSwitch(
     0=$Particles.PartitionParticles.Partition
     1=&Add_Integer(A=$Particles.PartitionParticles.Partition, B=2)
     2=?
     3=?
     4=?
     5=?
     Index=$Particles.SpawnGroup
     Num=::NewEnumerator0
    )
    Bone Sampling Mode=::NewEnumerator4
    Inherit Velocity=false
    Interpolate Mesh Sampling=false
    MaskSpawnGroup=false
    Sampled Position Offset=&MergeHelper_VectorSwitch(
     0=&MergeHelper_VectorSwitch(0=<-10 0 0>, 1=<10 0 0>, 2=?, 3=?, Index=$Particles.PartitionParticles.Partition, Num=::NewEnumerator0)
     1=&MergeHelper_VectorSwitch(0=<0 -10 0>, 1=<0 10 0>, 2=?, 3=?, Index=$Particles.PartitionParticles.Partition, Num=::NewEnumerator0)
     2=?
     3=?
     Index=$Particles.SpawnGroup
     Num=::NewEnumerator0
    )
    Sampled Position Offset Coordinate Space:ENiagaraMeshTransforms=::NewEnumerator3
    Skeletal Mesh=$System.Skeletal Mesh
   )
   /SetVariables(Particles.MeshOrientation=/Game/Effects/Niagara/Modules/MergeEmitters/MergeHelper_QuatSwitch(
     0:Quat4f=$Output.SkeletalMeshLocation.SampledRotation
     1:Quat4f=&MultiplyQuaternion(
      Quaternion A=$Output.SkeletalMeshLocation.SampledRotation
      Quaternion B=&MakeQuatFromAxisAngle(
       Axis=<0 0 1>
       AngleInDegrees=&MergeHelper_FloatSwitch(0=-90, 1=90, 2=?, 3=?, Index=$Particles.PartitionParticles.Partition, Num=::NewEnumerator0)
      )
     )
     2:Quat4f=?
     3:Quat4f=?
     4:Quat4f=?
     5:Quat4f=?
     6:Quat4f=?
     7:Quat4f=?
     8:Quat4f=?
     9:Quat4f=?
     Index:NiagaraInt32=$Particles.SpawnGroup
     Num:MergeHelper_SwitcherEnum=::NewEnumerator0
    )) {isSetParameters=true}
   &ScaleSpriteSize(Scale Factor=&MakeVector2D(
     X=&FloatFromCurve(
      FloatCurve=DataInterface() {propertyValues {Curve {keys [<at 0 0>,<at 1 2>]}}}
      Scale Curve=&MergeHelper_FloatSwitch(0=1.5, 1=1, 2=?, 3=?, Index=$Particles.SpawnGroup, Num=::NewEnumerator0)
     )
     Y=1
    ))
   /SetVariables(Particles.SpriteSize_3:Vector2f=&Multiply_Vector2(
     A=$Particles.Initial.SpriteSize_3
     B=&Vector2DFromFloat(Value=&FloatFromCurve(
       FloatCurve=DataInterface() {propertyValues {Curve {keys [<at 0 0>,<at 1 1>]}}}
       Scale Curve=&MergeHelper_FloatSwitch(0=1.5, 1=1, 2=?, 3=?, Index=$Particles.SpawnGroup, Num=::NewEnumerator0)
      ))
    )) {isSetParameters=true}
   &ScaleColor(
    Scale Alpha=&FloatFromCurve(FloatCurve=DataInterface() {propertyValues {Curve {keys [<at 0 1>,<at 1 0>]}}})
    Scale RGB=&MakeVectorFromLinearColorRGB(Color=&ColorFromCurve(ColorCurve=DataInterface() {propertyValues {
        AlphaCurve {keys []}
        BlueCurve {keys [<at 0 0.25 RCIM_Cubic>,<at 0.5 0.5>,<at 1 0.875 RCIM_Cubic>]}
        GreenCurve {keys [<at 0 0.525 RCIM_Cubic>,<at 0.5 0.6833>,<at 1 0.9208 RCIM_Cubic>]}
        RedCurve {keys [<at 0 1 RCIM_Cubic>,<at 0.5 1>,<at 1 1 RCIM_Cubic>]}
       }}))
    ScaleA=true
    ScaleRGB=true
   )
   &ScaleColor(
    Scale RGB=&VectorFromFloat(Value=&FloatFromCurve(FloatCurve=DataInterface() {propertyValues {Curve {keys [
          <at 0 100 RCIM_Cubic>
          {arriveTangent=-33.2952, interpMode=RCIM_Cubic, leaveTangent=-33.2924, tangentMode=RCTM_User, time=0.25, value=10}
          <at 1 0 RCIM_Cubic>
         ]}}}))
    ScaleA=true
    ScaleRGB=true
   ) {module=ScaleColor001}
   /SetVariables(
    Particles.Color_2:LinearColor=&MultiplyLinearColor(
     Linear Color=$Particles.Initial.Color_2
     Scale Factor=&MakeVector4_FromLinearColor(LinearColor=&MakeLinearColorFromVectorAndFloat(
       "Vector (RGB)"=&MakeVectorFromLinearColorRGB(Color=&ColorFromCurve(ColorCurve=DataInterface() {propertyValues {
           AlphaCurve {keys []}
           BlueCurve {keys [<at 0 0.25 RCIM_Cubic>,<at 0.5 0.5>,<at 1 0.875 RCIM_Cubic>]}
           GreenCurve {keys [<at 0 0.525 RCIM_Cubic>,<at 0.5 0.6833>,<at 1 0.9208 RCIM_Cubic>]}
           RedCurve {keys [<at 0 1 RCIM_Cubic>,<at 0.5 1>,<at 1 1 RCIM_Cubic>]}
          }}))
       "Float (Alpha)"=&FloatFromCurve(FloatCurve=DataInterface() {propertyValues {Curve {keys [<at 0 1>,<at 1 0>]}}})
      ))
    )
    Particles.Color_3:LinearColor=&MultiplyLinearColor(
     Linear Color=$Particles.Initial.Color_3
     Scale Factor=&MakeVector4_FromLinearColor(LinearColor=&MakeLinearColorFromVectorAndFloat(
       "Vector (RGB)"=&MakeVectorFromLinearColorRGB(Color=&ColorFromCurve(ColorCurve=DataInterface() {propertyValues {
           AlphaCurve {keys []}
           BlueCurve {keys [<at 0 0.25 RCIM_Cubic>,<at 0.5 0.5>,<at 1 0.875 RCIM_Cubic>]}
           GreenCurve {keys [<at 0 0.525 RCIM_Cubic>,<at 0.5 0.6833>,<at 1 0.9208 RCIM_Cubic>]}
           RedCurve {keys [<at 0 1 RCIM_Cubic>,<at 0.5 1>,<at 1 1 RCIM_Cubic>]}
          }}))
       "Float (Alpha)"=&FloatFromCurve(FloatCurve=DataInterface() {propertyValues {Curve {keys [<at 0 1>,<at 1 0>]}}})
      ))
    )
    Particles.Color_4:LinearColor=&MultiplyLinearColor(
     Linear Color=$Particles.Initial.Color_4
     Scale Factor=&MakeVector4_FromLinearColor(LinearColor=&MakeLinearColorFromVectorAndFloat(
       "Vector (RGB)"=&MakeVectorFromLinearColorRGB(Color=&ColorFromCurve(ColorCurve=DataInterface() {propertyValues {
           AlphaCurve {keys []}
           BlueCurve {keys [<at 0 0.25 RCIM_Cubic>,<at 0.5 0.5>,<at 1 0.875 RCIM_Cubic>]}
           GreenCurve {keys [<at 0 0.525 RCIM_Cubic>,<at 0.5 0.6833>,<at 1 0.9208 RCIM_Cubic>]}
           RedCurve {keys [<at 0 1 RCIM_Cubic>,<at 0.5 1>,<at 1 1 RCIM_Cubic>]}
          }}))
       "Float (Alpha)"=&FloatFromCurve(FloatCurve=DataInterface() {propertyValues {Curve {keys [<at 0 1>,<at 1 0>]}}})
      ))
    )
   ) {isSetParameters=true}
   /SetVariables(
    Particles.Color_2:LinearColor=&MultiplyLinearColor(
     Linear Color=$Particles.Color_2
     Scale Factor=&MakeFloat4FromVectorAndFloat(
      "Vector (RGB)"=&VectorFromFloat(Value=&FloatFromCurve(FloatCurve=DataInterface() {propertyValues {Curve {keys [<at 0 1 RCIM_Cubic>, {arriveTangent=-1, interpMode=RCIM_Cubic, leaveTangent=-1, time=0.1, value=6}, <at 1 0 RCIM_Cubic>]}}}))
      "Float (Alpha)"=1
     )
    )
    Particles.Color_4:LinearColor=&MultiplyLinearColor(
     Linear Color=$Particles.Initial.Color_4
     Scale Factor=&MakeFloat4FromVectorAndFloat(
      "Vector (RGB)"=&VectorFromFloat(Value=&FloatFromCurve(FloatCurve=DataInterface() {propertyValues {Curve {keys [
            <at 0 100 RCIM_Cubic>
            {arriveTangent=-36.3579, interpMode=RCIM_Cubic, leaveTangent=-36.3595, tangentMode=RCTM_User, time=0.25, value=10}
            <at 1 1 RCIM_Cubic>
           ]}}}))
      "Float (Alpha)"=1
     )
    )
   ) {isSetParameters=true}
   &ScaleMeshSize(Scale Factor=&VectorFromCurve(
     VectorCurve=DataInterface() {propertyValues {
       XCurve {keys [
         {arriveTangent=2.9897, interpMode=RCIM_Cubic, leaveTangent=2.9897, tangentMode=RCTM_User, time=0, value=0}
         {arriveTangent=3.9553, interpMode=RCIM_Cubic, leaveTangent=3.9553, tangentMode=RCTM_User, time=0.2, value=0.6}
         <at 1 2 RCIM_Cubic>
        ]}
       YCurve {keys [
         {arriveTangent=14.017, interpMode=RCIM_Cubic, leaveTangent=14.017, tangentMode=RCTM_User, time=0, value=0}
         <at 0.2 1 RCIM_Cubic>
         <at 1 0 RCIM_Cubic>
        ]}
       ZCurve {keys [
         {arriveTangent=13.3462, interpMode=RCIM_Cubic, leaveTangent=13.3462, tangentMode=RCTM_User, time=0, value=0}
         <at 0.2 1 RCIM_Cubic>
         <at 1 0 RCIM_Cubic>
        ]}
      }}
     Scale Curve=<1 2 2>
    ))
  ]
  rendererClasses [&NiagaraSpriteRendererProperties, &NiagaraMeshRendererProperties]
  renderers [
   {rendererClass=&NiagaraSpriteRendererProperties}
   {rendererClass=&NiagaraSpriteRendererProperties, rendererIndex=1}
   {rendererClass=&NiagaraSpriteRendererProperties, rendererIndex=2}
   {rendererClass=&NiagaraMeshRendererProperties, rendererIndex=3}
  ]
 }
 Calf_Flames {
  emitterUpdateScript [
   &EmitterState(
    Life Cycle Mode=::NewEnumerator1
    Loop Behavior=::NewEnumerator1
    Loop Delay=$System.SecondThrusterDelay
    Loop Duration=&Subtract_Float(A=$System.MainThrusterStart, B=$System.SecondThrusterDelay)
    UseLoopDelay=true
   )
   &SpawnBurst_Instantaneous(Spawn Count=2, Use Spawn Probability=false, UseLoopCountLimit=false)
   &SpawnRate(SpawnRate=30, Use Spawn Probability=false) {enabled=false}
   &SpawnRate(SpawnGroup=1, SpawnRate=40, Use Spawn Probability=false) {enabled=false, module=SpawnRate001}
   /SetVariables(Emitter.ScaleMultiplier=0.9) {isSetParameters=true, module=SetVariables_#7}
  ]
  particleSpawnScript [
   &PartitionParticles()
   /SetVariables(Particles.SpawnGroup=$Emitter.SpawnGroup) {isSetParameters=true, module=SetVariables_#8}
   /SetVariables(Particles.VisibilityTag=$Particles.SpawnGroup) {enabled=false, isSetParameters=true, module=SetVariables_#9}
   /Niagara/Modules/Spawn/Initialization/InitializeParticle(
    Color=<rgba 0.5928 2 0 0.2>
    Lifetime=0.3
    Mesh Scale=<0.2 1 1>
    Module.Write Color=true
    Module.Write Lifetime=true
    Module.Write Mass=true
    Module.Write Material Random=true
    Module.Write Position=true
    Module.Write Scale=true
    Module.Write SpriteRotation=false
    Module.Write SpriteSize=false
    Position=&SimulationPosition()
    Sprite Size=<25 25>
   ) {enabled=false}
   /Niagara/Modules/Spawn/Initialization/V2/InitializeParticle(
    Color=&MultiplyLinearColor(
     Linear Color=$System.LogoThrusterColor
     Scale Factor=&MakeVector4_FromLinearColor(LinearColor=&MergeHelper_ColorSwitch(0=<rgba 2 2 2 0.1>, 1=<rgba 2 2 2 0.2>, 2=?, 3=?, Index=$Particles.SpawnGroup, Num=::NewEnumerator0))
    )
    Lifetime=$Emitter.CurrentLoopDuration
    Mesh Scale=&MergeHelper_VectorSwitch(
     0=<0.2 1 1>
     1=&RandomRangeVector(Minimum=<0.1 0.2 0.3>, Maximum=<0.2 0.5 0.5>)
     2=?
     3=?
     Index=$Particles.SpawnGroup
     Num=::NewEnumerator0
    )
    Mesh Scale Mode=::NewEnumerator4
    Sprite Size=<40 20>
    Sprite Size Mode=::NewEnumerator4
    UsePositionOffset=false
    Write Material Random=true
   ) {module=InitializeParticle001}
   &FN_SetSpawnGroup(Number of Spawn Groups=5) {enabled=false}
  ]
  particleUpdateScript [
   &SolveForcesAndVelocity(Clamp Velocity=false, Limit Acceleration=false)
   &ParticleState()
   &SkeletalMeshLocation(
    Apply Sampled Orientation to Mesh Particles=true
    Bone / Socket Index=&Add_Integer(A=$Particles.PartitionParticles.Partition, B=2)
    Bone Sampling Mode=::NewEnumerator4
    Inherit Velocity=false
    Interpolate Mesh Sampling=false
    MaskSpawnGroup=false
    Sampled Position Offset=&MergeHelper_VectorSwitch(0=<0 -8 0>, 1=<0 8 0>, 2=?, 3=?, Index=$Particles.PartitionParticles.Partition, Num=::NewEnumerator0)
    Sampled Position Offset Coordinate Space:ENiagaraMeshTransforms=::NewEnumerator3
    Skeletal Mesh=$System.Skeletal Mesh
   )
   /SetVariables(Particles.MeshOrientation=&MultiplyQuaternion(
     Quaternion A=$Output.SkeletalMeshLocation.SampledRotation
     Quaternion B=&MergeHelper_QuatSwitch(
      0:Quat4f=&MakeQuatFromAxisAngle(Axis=<0 0 1>, AngleInDegrees=-90)
      1:Quat4f=&MakeQuatFromAxisAngle(Axis=<0 0 1>, AngleInDegrees=90)
      2:Quat4f=&MakeQuatFromAxisAngle(Axis=<0 1 0>, AngleInDegrees=-60)
      3:Quat4f=&MakeQuatFromAxisAngle(Axis=<0 1 0>, AngleInDegrees=60)
      4:Quat4f=?
      5:Quat4f=?
      6:Quat4f=?
      7:Quat4f=?
      8:Quat4f=?
      9:Quat4f=?
      Index:NiagaraInt32=$Particles.PartitionParticles.Partition
      Num:MergeHelper_SwitcherEnum=::NewEnumerator0
     )
    )) {isSetParameters=true, module=SetVariables_#10}
   &ScaleColor(
    Scale Alpha=&FloatFromCurve(FloatCurve=DataInterface() {propertyValues {Curve {keys [<at 0 0 RCIM_Cubic>,<at 0.5 1>,<at 1.0011 0 RCIM_Cubic>]}}})
    Scale RGB=&VectorFromCurve(VectorCurve=DataInterface() {propertyValues {
       XCurve {keys [<at 0 2 RCIM_Cubic>,<at 1 1 RCIM_Cubic>]}
       YCurve {keys [<at 0 2 RCIM_Cubic>,<at 1 1 RCIM_Cubic>]}
       ZCurve {keys [<at 0 2 RCIM_Cubic>,<at 1 1 RCIM_Cubic>]}
      }})
    ScaleA=true
    ScaleRGB=true
   ) {enabled=false}
   &DynamicMaterialParameters(
    Index 0 Param 1=&FloatFromCurve(FloatCurve=DataInterface() {propertyValues {Curve {keys [<at 0 1>,<at 1 0>]}}})
    Index 0 Param 2=5
    Index 0 Param 3=20
    Index 0 Param 4=1
    Module.Param0WriteEnabled=true
    Module.Param1WriteEnabled=true
    Module.Param2WriteEnabled=true
    Module.Param3WriteEnabled=true
   )
   &ScaleMeshSize(Scale Factor=&VectorFromFloat(Value=&Multiply_Float(
      A=&Waveform("Waveform [1]"=::NewEnumerator12, "[1] Frequency"=25, "[1] Amplitude Min/Max"=<0.75 1.25>)
      B=$Emitter.ScaleMultiplier
     )))
   &CameraOffset()
   &SpriteFacingAndAlignment(Sprite Facing=<0 1 0>, WriteAlignment=false, WriteFacing=true) {enabled=false}
   &AlignSpriteToMeshOrientation(
    Mesh Orientation Relative Sprite Alignment Vector=<1 0 0>
    Mesh Orientation Relative Sprite Facing Vector:Vector3f=<0 0 1>
    SetAlignment=true
    SetFacing=true
   )
   &ScaleSpriteSize(
    Scale Sprite Size Mode=::NewEnumerator0
    Uniform Scale Factor=&Multiply_Float(
     A=&Waveform(
      Waveform Count=::NewEnumerator1
      "Waveform [1]"=::NewEnumerator12
      "[1] Frequency"=12
      "[1] Amplitude Min/Max"=<0.9 1.1>
      "Waveform [2] Blend"=::NewEnumerator2
      "Waveform [2]"=::NewEnumerator12
      "[2] Input Position"=0
      "[2] Phase"=$Emitter.Age
      "[2] Phase Scale"=1
      "[2] Compound Frequency"=2
      "[2] Frequency"=20
      "[2] Amplitude Min/Max"=<0.9 1.1>
      "[2] Scale"=1
      "[2] Bias"=0
      "[2] Exponent"=1
      "[2] Pulse Duration"=0.1
      "[2] Clip Triangle Min/Max"=<0 0>
      "[2] Random Seed"=0
     )
     B=$Emitter.ScaleMultiplier
    )
   )
   &FN_Sprite_Attributes(
    Color=&Multiply_LinearColorByFloat(
     Linear Color=$System.LogoThrusterColor
     Float=&Waveform("Waveform [1]"=::NewEnumerator12, "[1] Frequency"=12, "[1] Amplitude Min/Max"=<0.8 1.1>, "[1] Scale"=1.7)
    )
    Create Color Attribute=true
    Create SpriteSize Attribute=true
    SpriteSize=<100 100>
    Write to Module Namespace=true
   )
  ]
  rendererClasses [&NiagaraMeshRendererProperties, &NiagaraSpriteRendererProperties]
  renderers [
   {rendererClass=&NiagaraMeshRendererProperties, rendererIndex=0}
   {rendererClass=&NiagaraMeshRendererProperties, rendererIndex=1}
   {rendererClass=&NiagaraMeshRendererProperties, rendererIndex=2}
   {rendererClass=&NiagaraSpriteRendererProperties, rendererIndex=3}
   {rendererClass=&NiagaraSpriteRendererProperties, rendererIndex=4}
  ]
 }
 Flame_Launch {
  emitterUpdateScript [
   &EmitterState(
    Life Cycle Mode=::NewEnumerator1
    Loop Behavior=::NewEnumerator1
    Loop Delay=$System.FirstThrusterDelay
    Loop Duration=&Subtract_Float(A=$System.MainThrusterStart, B=$System.FirstThrusterDelay)
    UseLoopDelay=true
   )
   &SpawnBurst_Instantaneous(Spawn Count=2, Spawn Group=2, Use Spawn Probability=false, UseLoopCountLimit=false)
   &SpawnRate(SpawnRate=30, Use Spawn Probability=false)
   &SpawnRate(SpawnGroup=1, SpawnRate=40, Use Spawn Probability=false) {module=SpawnRate001}
  ]
  particleSpawnScript [
   &PartitionParticles()
   /SetVariables(Particles.RandomSide:NiagaraInt32=&RandomRangeInt(Minimum=0, Maximum=1)) {isSetParameters=true}
   /SetVariables(Particles.SpawnGroup=$Emitter.SpawnGroup) {isSetParameters=true, module=SetVariables_#8}
   /SetVariables(Particles.VisibilityTag=$Particles.SpawnGroup) {isSetParameters=true, module=SetVariables_#9}
   /Niagara/Modules/Spawn/Initialization/InitializeParticle(
    Color=<rgba 0.5928 2 0 0.2>
    Lifetime=0.3
    Mesh Scale=<0.2 1 1>
    Module.Write Color=true
    Module.Write Lifetime=true
    Module.Write Mass=true
    Module.Write Material Random=true
    Module.Write Position=true
    Module.Write Scale=true
    Module.Write SpriteRotation=false
    Module.Write SpriteSize=false
    Position=&SimulationPosition()
    Sprite Size=<25 25>
   ) {enabled=false}
   /Niagara/Modules/Spawn/Initialization/V2/InitializeParticle(
    Color=&MultiplyLinearColor(
     Linear Color=$System.LogoThrusterColor
     Scale Factor=&MakeVector4_FromLinearColor(LinearColor=&MergeHelper_ColorSwitch(0=<rgba 1.2 1.2 1.2 0.2>, 1=<rgba 1.5 1.5 1.5 0.2>, 2=?, 3=?, Index=$Particles.SpawnGroup, Num=::NewEnumerator0))
    )
    Lifetime=&MergeHelper_FloatSwitch(0=0.3, 1=0.1, 2=?, 3=?, Index=$Particles.SpawnGroup, Num=::NewEnumerator0)
    Mesh Scale=&MergeHelper_VectorSwitch(
     0=<0.2 1 1>
     1=&RandomRangeVector(Minimum=<0.1 0.2 0.3>, Maximum=<0.2 0.5 0.5>)
     2=?
     3=?
     Index=$Particles.SpawnGroup
     Num=::NewEnumerator0
    )
    Mesh Scale Mode=::NewEnumerator4
    Sprite Rotation Mode=::NewEnumerator2
    Sprite Size Mode=::NewEnumerator3
    Uniform Sprite Size=80
    UsePositionOffset=false
    Write Material Random=true
   ) {module=InitializeParticle001}
   &FN_SetSpawnGroup(Number of Spawn Groups=5) {enabled=false}
  ]
  particleUpdateScript [
   &ParticleState()
   &SkeletalMeshLocation(
    Apply Sampled Orientation to Mesh Particles=true
    Bone / Socket Index=$Particles.PartitionParticles.Partition
    Bone Sampling Mode=::NewEnumerator4
    Inherit Velocity=false
    Interpolate Mesh Sampling=false
    MaskSpawnGroup=false
    Sampled Position Offset=<0 0 -7>
    Sampled Position Offset Coordinate Space:ENiagaraMeshTransforms=::NewEnumerator1
    Skeletal Mesh=$System.Skeletal Mesh
   )
   /SetVariables(Particles.Position=&MergeHelper_VectorSwitch(0=$System.LeftFootPos, 1=$System.RightFootPos, 2=?, 3=?, Index=$Particles.PartitionParticles.Partition, Num=::NewEnumerator0)) {enabled=false, isSetParameters=true}
   /SetVariables(Particles.MeshOrientation=&MultiplyQuaternion(
     Quaternion A=$Output.SkeletalMeshLocation.SampledRotation
     Quaternion B=&MergeHelper_QuatSwitch(
      0:Quat4f=&MakeQuatFromAxisAngle(
       Axis=<0 1 0>
       AngleInDegrees=&MergeHelper_FloatSwitch(0=60, 1=-70, 2=?, 3=?, Index=$Particles.RandomSide, Num=::NewEnumerator0)
      )
      1:Quat4f=&MakeQuatFromAxisAngle(
       Axis=<0 1 0>
       AngleInDegrees=&MergeHelper_FloatSwitch(0=-120, 1=110, 2=?, 3=?, Index=$Particles.RandomSide, Num=::NewEnumerator0)
      )
      2:Quat4f=&MakeQuatFromAxisAngle(Axis=<0 1 0>, AngleInDegrees=-60)
      3:Quat4f=&MakeQuatFromAxisAngle(Axis=<0 1 0>, AngleInDegrees=60)
      4:Quat4f=?
      5:Quat4f=?
      6:Quat4f=?
      7:Quat4f=?
      8:Quat4f=?
      9:Quat4f=?
      Index:NiagaraInt32=$Particles.PartitionParticles.Partition
      Num:MergeHelper_SwitcherEnum=::NewEnumerator0
     )
    )) {isSetParameters=true, module=SetVariables_#10}
   &ScaleColor(
    Scale Alpha=&FloatFromCurve(FloatCurve=DataInterface() {propertyValues {Curve {keys [<at 0 0 RCIM_Cubic>,<at 0.5 1>,<at 1.0011 0 RCIM_Cubic>]}}})
    Scale RGB=&VectorFromCurve(VectorCurve=DataInterface() {propertyValues {
       XCurve {keys [<at 0 2 RCIM_Cubic>,<at 1 1 RCIM_Cubic>]}
       YCurve {keys [<at 0 2 RCIM_Cubic>,<at 1 1 RCIM_Cubic>]}
       ZCurve {keys [<at 0 2 RCIM_Cubic>,<at 1 1 RCIM_Cubic>]}
      }})
    ScaleA=true
    ScaleRGB=true
   )
   &DynamicMaterialParameters(
    Index 0 Param 1=&FloatFromCurve(FloatCurve=DataInterface() {propertyValues {Curve {keys [<at 0 1>,<at 1 0>]}}})
    Index 0 Param 2=5
    Index 0 Param 3=20
    Index 0 Param 4=1
    Module.Param0WriteEnabled=true
    Module.Param1WriteEnabled=true
    Module.Param2WriteEnabled=true
    Module.Param3WriteEnabled=true
   )
   &ScaleMeshSize(Scale Factor=&VectorFromCurve(
     VectorCurve=DataInterface() {propertyValues {
       XCurve {keys [
         <at 0 0.75 RCIM_Cubic>
         <at 0.4907 3.9579>
         {arriveTangent=4.355, interpMode=RCIM_Cubic, leaveTangent=4.355, tangentMode=RCTM_User, time=0.5, value=4}
         <at 1 0.5 RCIM_Cubic>
        ]}
       YCurve {keys [
         <at 0 1 RCIM_Cubic>
         <at 0.4907 0.9933>
         {arriveTangent=0.75, interpMode=RCIM_Cubic, leaveTangent=0.75, time=0.5, value=1}
         <at 1 1.75 RCIM_Cubic>
        ]}
       ZCurve {keys [<at -0.0014 1 RCIM_Cubic>, {arriveTangent=0.7489, interpMode=RCIM_Cubic, leaveTangent=0.7489, time=0.5, value=1}, <at 1 1.75 RCIM_Cubic>]}
      }}
     Scale Curve=&VectorFromFloat(Value=&FloatFromCurve(
       FloatCurve=DataInterface() {propertyValues {Curve {keys [
           <at 0 1.025 RCIM_Cubic>
           {arriveTangent=-0.1071, interpMode=RCIM_Cubic, leaveTangent=-0.1071, time=0.3, value=1.1}
           {arriveTangent=0.5714, interpMode=RCIM_Cubic, leaveTangent=0.5714, time=0.7, value=0.95}
           <at 1 1.5 RCIM_Cubic>
          ]}}}
       CurveIndex=$Emitter.NormalizedLoopAge
       Scale Curve=0.85
      ))
    ))
   &CameraOffset(Camera Offset Amount=&MergeHelper_FloatSwitch(0=0, 1=0, 2=50, 3=0, 4=0, Index=$Particles.SpawnGroup, Num=::NewEnumerator3))
  ]
  rendererClasses [&NiagaraMeshRendererProperties, &NiagaraSpriteRendererProperties]
  renderers [
   {rendererClass=&NiagaraMeshRendererProperties, rendererIndex=0}
   {rendererClass=&NiagaraMeshRendererProperties, rendererIndex=1}
   {rendererClass=&NiagaraSpriteRendererProperties, rendererIndex=2}
  ]
 }
 FootThrusters {
  emitterUpdateScript [
   &EmitterState(
    Inactive Response=::NewEnumerator1
    Life Cycle Mode=::NewEnumerator1
    Loop Behavior=::NewEnumerator1
    Loop Delay=$System.MainThrusterStart
    Loop Duration=$System.ThrusterCutoff
    UseLoopDelay=true
   )
   &SpawnBurst_Instantaneous(Spawn Count=2, Use Spawn Probability=false, UseLoopCountLimit=false)
   &SpawnRate(SpawnRate=30, Use Spawn Probability=false) {enabled=false}
   &SpawnRate(SpawnGroup=1, SpawnRate=40, Use Spawn Probability=false) {enabled=false, module=SpawnRate001}
   /SetVariables(Emitter.ScaleMultiplier=1) {isSetParameters=true, module=SetVariables_#7}
   /SetVariables(Emitter.SpriteSizeScaler:NiagaraFloat=&Multiply_Float(
     A=&Waveform(
      Waveform Count=::NewEnumerator1
      "Waveform [1]"=::NewEnumerator12
      "[1] Frequency"=12
      "[1] Amplitude Min/Max"=<0.9 1.1>
      "Waveform [2] Blend"=::NewEnumerator2
      "Waveform [2]"=::NewEnumerator12
      "[2] Input Position"=0
      "[2] Phase"=$Emitter.Age
      "[2] Phase Scale"=1
      "[2] Compound Frequency"=2
      "[2] Frequency"=20
      "[2] Amplitude Min/Max"=<0.9 1.1>
      "[2] Scale"=1
      "[2] Bias"=0
      "[2] Exponent"=1
      "[2] Pulse Duration"=0.1
      "[2] Clip Triangle Min/Max"=<0 0>
      "[2] Random Seed"=0
     )
     B=$Emitter.ScaleMultiplier
    )) {isSetParameters=true}
  ]
  particleSpawnScript [
   &PartitionParticles()
   /SetVariables(Particles.SpawnGroup=$Emitter.SpawnGroup) {isSetParameters=true, module=SetVariables_#8}
   /SetVariables(Particles.VisibilityTag=$Particles.SpawnGroup) {enabled=false, isSetParameters=true, module=SetVariables_#9}
   /Niagara/Modules/Spawn/Initialization/InitializeParticle(
    Color=<rgba 0.5928 2 0 0.2>
    Lifetime=0.3
    Mesh Scale=<0.2 1 1>
    Module.Write Color=true
    Module.Write Lifetime=true
    Module.Write Mass=true
    Module.Write Material Random=true
    Module.Write Position=true
    Module.Write Scale=true
    Module.Write SpriteRotation=false
    Module.Write SpriteSize=false
    Position=&SimulationPosition()
    Sprite Size=<25 25>
   ) {enabled=false}
   /Niagara/Modules/Spawn/Initialization/V2/InitializeParticle(
    Color=&MultiplyLinearColor(
     Linear Color=$System.LogoThrusterColor
     Scale Factor=&MakeVector4_FromLinearColor(LinearColor=&MergeHelper_ColorSwitch(0=<rgba 2 2 2 0.1>, 1=<rgba 2 2 2 0.2>, 2=?, 3=?, Index=$Particles.SpawnGroup, Num=::NewEnumerator0))
    )
    Lifetime=$Emitter.CurrentLoopDuration
    Mesh Scale=&MergeHelper_VectorSwitch(
     0=<0.2 1 1>
     1=&RandomRangeVector(Minimum=<0.1 0.2 0.3>, Maximum=<0.2 0.5 0.5>)
     2=?
     3=?
     Index=$Particles.SpawnGroup
     Num=::NewEnumerator0
    )
    Mesh Scale Mode=::NewEnumerator4
    Sprite Size=<50 25>
    Sprite Size Mode=::NewEnumerator4
    UsePositionOffset=false
    Write Material Random=true
   ) {module=InitializeParticle001}
   /SetVariables(Particles.LogoSpriteSize:Vector2f=<20 40>, Particles.LogoThrusterColor:LinearColor=$System.LogoThrusterColor) {isSetParameters=true}
   &FN_SetSpawnGroup(Number of Spawn Groups=5) {enabled=false}
  ]
  particleUpdateScript [
   &SolveForcesAndVelocity(Clamp Velocity=false, Limit Acceleration=false)
   &ParticleState()
   &SkeletalMeshLocation(
    Bone / Socket Index=&Add_Integer(A=$Particles.PartitionParticles.Partition, B=0)
    Bone Sampling Mode=::NewEnumerator4
    Inherit Velocity=false
    Interpolate Mesh Sampling=false
    MaskSpawnGroup=false
    Sampled Position Offset=<0 0 -10>
    Sampled Position Offset Coordinate Space:ENiagaraMeshTransforms=::NewEnumerator1
    Skeletal Mesh=$System.Skeletal Mesh
   ) {enabled=false}
   &UpdateMeshOrientation(
    Facing Direction=<0 0 -1>
    Facing Interpolation=false
    Orientation Method=::NewEnumerator1
    Override Quat=false
    TransformFacing=false
    TransformUp=false
    Up Direction:Vector3f=<1 0 0>
   )
   &UpdateMeshOrientation(Rotation Rate=8) {module=UpdateMeshOrientation001}
   /SetVariables(Particles.Position=&MergeHelper_VectorSwitch(0=$System.LeftFootPos, 1=$System.RightFootPos, 2=?, 3=?, Index=$Particles.PartitionParticles.Partition, Num=::NewEnumerator0)) {isSetParameters=true}
   /SetVariables(Particles.MeshOrientation=&MakeQuatFromAxisAngle(Axis=<0 1 0>, AngleInDegrees=0)) {enabled=false, isSetParameters=true, module=SetVariables_#10}
   &ScaleColor(
    Scale Alpha=&FloatFromCurve(FloatCurve=DataInterface() {propertyValues {Curve {keys [<at 0 0 RCIM_Cubic>,<at 0.5 1>,<at 1.0011 0 RCIM_Cubic>]}}})
    Scale RGB=&VectorFromCurve(VectorCurve=DataInterface() {propertyValues {
       XCurve {keys [<at 0 2 RCIM_Cubic>,<at 1 1 RCIM_Cubic>]}
       YCurve {keys [<at 0 2 RCIM_Cubic>,<at 1 1 RCIM_Cubic>]}
       ZCurve {keys [<at 0 2 RCIM_Cubic>,<at 1 1 RCIM_Cubic>]}
      }})
    ScaleA=true
    ScaleRGB=true
   ) {enabled=false}
   &DynamicMaterialParameters(
    Index 0 Param 1=&FloatFromCurve(FloatCurve=DataInterface() {propertyValues {Curve {keys [<at 0 1>,<at 1 0>]}}})
    Index 0 Param 2=5
    Index 0 Param 3=20
    Index 0 Param 4=1
    Module.Param0WriteEnabled=true
    Module.Param1WriteEnabled=true
    Module.Param2WriteEnabled=true
    Module.Param3WriteEnabled=true
   )
   &ScaleMeshSize(Scale Factor=&VectorFromFloat(Value=&Multiply_Float(
      A=&Waveform("Waveform [1]"=::NewEnumerator12, "[1] Frequency"=35, "[1] Amplitude Min/Max"=<0.75 1.25>)
      B=$Emitter.ScaleMultiplier
     )))
   &CameraOffset()
   &SpriteFacingAndAlignment(Sprite Facing=<0 1 0>, WriteAlignment=false, WriteFacing=true) {enabled=false}
   /SetVariables(
    Particles.LogoSpriteSize:Vector2f=&Multiply_Vector2D_ByFloat(Float=$Emitter.SpriteSizeScaler, Vector2D=$Particles.Initial.LogoSpriteSize)
    Particles.LogoThrusterColor:LinearColor=&MultiplyLinearColor(Linear Color=$Particles.Initial.LogoThrusterColor, Scale Factor=<1 1 1 1>)
   ) {isSetParameters=true}
   &AlignSpriteToMeshOrientation(
    Mesh Orientation Relative Sprite Alignment Vector=<1 0 0>
    Mesh Orientation Relative Sprite Facing Vector:Vector3f=<0 0 1>
    SetAlignment=true
    SetFacing=true
   )
   &ScaleSpriteSize(
    Scale Sprite Size Mode=::NewEnumerator0
    Uniform Scale Factor=&Multiply_Float(
     A=&Waveform(
      Waveform Count=::NewEnumerator1
      "Waveform [1]"=::NewEnumerator12
      "[1] Frequency"=12
      "[1] Amplitude Min/Max"=<0.9 1.1>
      "Waveform [2] Blend"=::NewEnumerator2
      "Waveform [2]"=::NewEnumerator12
      "[2] Input Position"=0
      "[2] Phase"=$Emitter.Age
      "[2] Phase Scale"=1
      "[2] Compound Frequency"=2
      "[2] Frequency"=20
      "[2] Amplitude Min/Max"=<0.9 1.1>
      "[2] Scale"=1
      "[2] Bias"=0
      "[2] Exponent"=1
      "[2] Pulse Duration"=0.1
      "[2] Clip Triangle Min/Max"=<0 0>
      "[2] Random Seed"=0
     )
     B=$Emitter.ScaleMultiplier
    )
   )
   &FN_Sprite_Attributes(
    Color=&Multiply_LinearColorByFloat(
     Linear Color=<rgba 0.2443 1.5 0.05 1>
     Float=&Waveform("Waveform [1]"=::NewEnumerator12, "[1] Frequency"=12, "[1] Amplitude Min/Max"=<0.8 1.25>, "[1] Scale"=0.35)
    )
    Create Color Attribute=true
    Create SpriteSize Attribute=true
    SpriteSize=<200 200>
    Write to Module Namespace=true
   )
   &FN_Mesh_Attributes(Color=<rgba 1 1 1 0.01>, Create Color Attribute=true, Write to Module Namespace=true)
  ]
  rendererClasses [&NiagaraMeshRendererProperties, &NiagaraSpriteRendererProperties]
  renderers [
   {rendererClass=&NiagaraMeshRendererProperties, rendererIndex=0}
   {rendererClass=&NiagaraMeshRendererProperties, rendererIndex=1}
   {rendererClass=&NiagaraSpriteRendererProperties, rendererIndex=2}
   {rendererClass=&NiagaraMeshRendererProperties, rendererIndex=3}
   {rendererClass=&NiagaraMeshRendererProperties, rendererIndex=4}
  ]
 }
 Rim_Glow {
  emitterSpawnScript [/SetVariables(Emitter.SpawnTimeOffset=0.33) {isSetParameters=true}]
  emitterUpdateScript [
   /SetVariables(Emitter.GroundToAirT:NiagaraFloat=&FloatFromCurve(
     FloatCurve=DataInterface() {propertyValues {Curve {keys [<at 0 0>, {arriveTangent=0.5, interpMode=RCIM_Cubic, leaveTangent=0.5, time=1.6, value=0}, <at 2 1 RCIM_Cubic>]}}}
     CurveIndex=$System.Age
    )) {isSetParameters=true}
   &EmitterState(Inactive Response=::NewEnumerator1, Life Cycle Mode=::NewEnumerator1, Loop Behavior=::NewEnumerator1, UseLoopDelay=false)
   &SpawnBurst_Instantaneous(Spawn Count=1, Spawn Time=$Emitter.SpawnTimeOffset, Use Spawn Probability=false)
  ]
  particleSpawnScript [
   &PartitionParticles()
   &FN_SetSpawnGroup_v2(Number of Spawn Groups=4) {enabled=false}
   /Niagara/Modules/Spawn/Initialization/InitializeParticle(
    Color=<rgba 0.5928 2 0 1>
    Lifetime=5
    Module.Write Color=true
    Module.Write Lifetime=false
    Module.Write Mass=true
    Module.Write Material Random=true
    Module.Write Position=false
    Module.Write Scale=false
    Module.Write SpriteRotation=false
    Module.Write SpriteSize=true
    Position=&SimulationPosition()
    Sprite Size=<4 4>
   ) {enabled=false}
   /Niagara/Modules/Spawn/Initialization/V2/InitializeParticle(
    Color=<rgba 0.5928 2 0 1>
    Lifetime=&Subtract_Float(A=3.75, B=$Emitter.SpawnTimeOffset)
    UsePositionOffset=false
    Write Material Random=true
   ) {module=InitializeParticle001}
   &CameraOffset(Camera Offset Amount=3)
  ]
  particleUpdateScript [
   &ParticleState()
   &SkeletalMeshLocation(
    Apply Sampled Orientation to Mesh Particles=true
    Bone Sampling Mode=::NewEnumerator4
    Inherit Velocity=false
    Interpolate Mesh Sampling=false
    MaskSpawnGroup=false
    Sampled Position Offset=<0 0 -110>
    Sampled Position Offset Coordinate Space:ENiagaraMeshTransforms=::NewEnumerator1
    Skeletal Mesh=$System.Skeletal Mesh
   ) {enabled=false}
   &SkeletalMeshLocation(
    Apply Sampled Orientation to Mesh Particles=true
    Bone / Socket Index=1
    Bone Sampling Mode=::NewEnumerator4
    Inherit Velocity=false
    Interpolate Mesh Sampling=false
    MaskSpawnGroup=false
    Sampled Position Offset=<0 0 -110>
    Sampled Position Offset Coordinate Space:ENiagaraMeshTransforms=::NewEnumerator1
    Skeletal Mesh=$System.Skeletal Mesh
   ) {enabled=false, module=SkeletalMeshLocation001}
   /SetVariables(Particles.Position=&AddVectorToPosition(
     Position=&LerpPosition(A=$System.LeftFootPos, B=$System.RightFootPos, Alpha=0.5)
     Vector=&MakeVector(
      X=0
      Y=0
      Z=&FloatFromCurve(FloatCurve=DataInterface() {propertyValues {Curve {keys [<at 0 20>,<at 1.1 -30>]}}}, CurveIndex=$Emitter.GroundToAirT)
     )
    )) {isSetParameters=true}
   &FN_Light_v3(
    Color=<rgba 0.5249 2 0.0772 1>
    Create Color Attribute=true
    Create Exponent Attribute=true
    Create Radius Attribute=true
    Enable Particle Light Enable=true
    Exponent=4
    Light Enabled=true
    Radius:NiagaraFloat=&Multiply_Float(
     A=&FloatFromCurve(
      FloatCurve=DataInterface() {propertyValues {Curve {keys [
          {arriveTangent=23.3484, interpMode=RCIM_Cubic, time=0, value=0}
          {arriveTangent=3.3333, interpMode=RCIM_Cubic, leaveTangent=3.3333, time=0.1, value=2}
          {arriveTangent=-2.0241, interpMode=RCIM_Cubic, leaveTangent=-2.0241, tangentMode=RCTM_User, time=0.3, value=1}
          <at 1.1 1 RCIM_Cubic>
         ]}}}
      CurveIndex=&Multiply_Float(A=$Particles.NormalizedAge, B=10)
      Scale Curve=65
     )
     B=&Waveform("Waveform [1]"=::NewEnumerator12, "[1] Frequency"=25, "[1] Amplitude Min/Max"=<0.9 1.1>)
    )
    Scale Color:NiagaraFloat=&FloatFromCurve(
     FloatCurve=DataInterface() {propertyValues {Curve {keys [
         <at 0 1 RCIM_Cubic>
         <at 0.1 2>
         {arriveTangent=-1.1765, interpMode=RCIM_Cubic, leaveTangent=-1.1765, time=0.3, value=1}
         {arriveTangent=-1.4286, interpMode=RCIM_Cubic, leaveTangent=-1.4286, time=0.95, value=1}
         <at 1 0 RCIM_Cubic>
        ]}}}
     Scale Curve=5
    )
   )
   /SetVariables(Particles.SpriteSize=&Multiply_Vector2D_ByFloat(Float=$Particles.FN_Light_v3.Radius, Vector2D=<1 1>)) {isSetParameters=true}
  ]
  rendererClasses [&NiagaraSpriteRendererProperties, &NiagaraLightRendererProperties]
  renderers [
   {rendererClass=&NiagaraSpriteRendererProperties}
   {rendererClass=&NiagaraLightRendererProperties, rendererIndex=1}
  ]
 }
 Smoke {
  emitterUpdateScript [
   &EmitterState(
    Inactive Response=::NewEnumerator1
    Life Cycle Mode=::NewEnumerator1
    Loop Behavior=::NewEnumerator1
    Loop Delay=$System.MainThrusterStart
    Loop Duration=$System.ThrusterCutoff
    UseLoopDelay=true
   )
   &SpawnRate(SpawnRate=80, Use Spawn Probability=false)
  ]
  particleSpawnScript [
   &PartitionParticles()
   /Niagara/Modules/Spawn/Initialization/InitializeParticle(
    Color=<rgba 1 1 1 0.0625>
    Lifetime=&UniformRangedFloat(Minimum=0.45, Maximum=0.65)
    Module.Write Color=true
    Module.Write Lifetime=true
    Module.Write Mass=true
    Module.Write Material Random=true
    Module.Write Position=false
    Module.Write Scale=false
    Module.Write SpriteRotation=true
    Module.Write SpriteSize=true
    Position=&SimulationPosition()
    Sprite Size=&Vector2DFromFloat(Value=&Lerp_Float(
      A=&UniformRangedFloat(Minimum=13, Maximum=25)
      B=&UniformRangedFloat(Minimum=25, Maximum=50)
      Alpha=0.5
     ))
   ) {enabled=false}
   /Niagara/Modules/Spawn/Initialization/V2/InitializeParticle(
    Color=&MultiplyLinearColor(
     Linear Color=$System.LogoThrusterColor
     Scale Factor=&MakeVector4_FromLinearColor(LinearColor=<rgba 5 5 5 0.02>)
    )
    Lifetime=&UniformRangedFloat(Minimum=0.45, Maximum=0.65)
    Sprite Rotation Mode=::NewEnumerator2
    Sprite Size Mode=::NewEnumerator1
    Uniform Sprite Size Max=35
    Uniform Sprite Size Min=20
    UsePositionOffset=false
    Write Material Random=true
   ) {module=InitializeParticle001}
   &DynamicMaterialParameters(
    Index 0 Param 2=1
    Index 0 Param 4=1
    Module.Param0WriteEnabled=false
    Module.Param1WriteEnabled=true
    Module.Param2WriteEnabled=false
    Module.Param3WriteEnabled=true
   )
  ]
  particleUpdateScript [
   &ParticleState()
   &SkeletalMeshLocation(
    Bone / Socket Index=$Particles.PartitionParticles.Partition
    Bone Sampling Mode=::NewEnumerator4
    Inherit Velocity=false
    MaskSpawnGroup=false
    Sampled Position Offset=<0 0 -7>
    Sampled Position Offset Coordinate Space:ENiagaraMeshTransforms=::NewEnumerator1
    Skeletal Mesh=$System.Skeletal Mesh
   ) {enabled=false}
   /SetVariables(Particles.Position=&MergeHelper_VectorSwitch(0=$System.LeftFootPos, 1=$System.RightFootPos, 2=?, 3=?, Index=$Particles.PartitionParticles.Partition, Num=::NewEnumerator0)) {isSetParameters=true}
   &AddVelocity(Velocity=&UniformRangedVector(Minimum=<0 0 -50>, Maximum=<0 0 -100>))
   &AccelerationForce(Acceleration=<0 0 100>)
   &Drag(Use Linear Drag=true, Use Rotational Drag=false)
   &SolveForcesAndVelocity(Clamp Velocity=false, Limit Acceleration=false)
   &DynamicMaterialParameters(
    Index 0 Param 1=&Lerp_Float(A=1, B=-1, Alpha=$Particles.NormalizedAge)
    Index 0 Param 2=1
    Index 0 Param 4=1
    Module.Param0WriteEnabled=true
    Module.Param1WriteEnabled=false
    Module.Param2WriteEnabled=false
    Module.Param3WriteEnabled=false
   ) {module=DynamicMaterialParameters001}
   /SetVariables(Particles.SubImageIndex=&FloatFromCurve(FloatCurve=DataInterface() {propertyValues {Curve {keys [{arriveTangent=105.1181, interpMode=RCIM_Cubic, leaveTangent=105.1181, tangentMode=RCTM_User, time=0, value=0}, <at 1 63 RCIM_Cubic>]}}})) {isSetParameters=true}
   &ScaleColor(
    Scale Alpha=&FloatFromCurve(FloatCurve=DataInterface() {propertyValues {Curve {keys [<at 0 0 RCIM_Cubic>,<at 0.1 1 RCIM_Cubic>,<at 1 0 RCIM_Cubic>]}}})
    Scale RGB=&VectorFromFloat(Value=&FloatFromCurve(FloatCurve=DataInterface() {propertyValues {Curve {keys [<at 0 5 RCIM_Cubic>,<at 0.5 1 RCIM_Cubic>]}}}))
    ScaleA=true
    ScaleRGB=true
   )
   &ScaleColor(
    Scale RGB=&MakeVectorFromLinearColorRGB(Color=&ColorFromCurve(ColorCurve=DataInterface() {propertyValues {
        AlphaCurve {keys []}
        BlueCurve {keys [<at 0 0.25 RCIM_Cubic>,<at 0.6182 0.125 RCIM_Cubic>]}
        GreenCurve {keys [<at 0 0.4375 RCIM_Cubic>,<at 0.6182 0.125 RCIM_Cubic>]}
        RedCurve {keys [<at 0 1 RCIM_Cubic>,<at 0.6182 0.125 RCIM_Cubic>]}
       }}))
    ScaleA=true
    ScaleRGB=true
   ) {module=ScaleColor001}
   &ScaleSpriteSize(Scale Factor=&Vector2DFromFloat(Value=&FloatFromCurve(FloatCurve=DataInterface() {propertyValues {Curve {keys [{arriveTangent=6.2193, interpMode=RCIM_Cubic, leaveTangent=6.2193, tangentMode=RCTM_User, time=0, value=0.2}, <at 1 2 RCIM_Cubic>]}}})))
  ]
  rendererClasses [&NiagaraSpriteRendererProperties]
  renderers [{rendererClass=&NiagaraSpriteRendererProperties}]
 }
 Spark_Start {
  simTarget=GPUComputeSim
  emitterUpdateScript [
   &EmitterState(
    Life Cycle Mode=::NewEnumerator1
    Loop Behavior=::NewEnumerator1
    Loop Delay=$System.MainThrusterStart
    Loop Duration=0.25
    UseLoopDelay=true
   )
   &SpawnRate(SpawnRate=250, Use Spawn Probability=false)
   &SpawnBurst_Instantaneous(
    Spawn Count=&UniformRangedInt(Minimum=40, Maximum=80)
    Spawn Time=$System.MainThrusterStart
    Use Spawn Probability=false
   ) {enabled=false}
  ]
  particleSpawnScript [
   &PartitionParticles()
   /SetVariables(Particles.SpawnGroup=$Emitter.SpawnGroup) {isSetParameters=true}
   /Niagara/Modules/Spawn/Initialization/InitializeParticle(
    Color=&MultiplyLinearColor(
     Linear Color=$System.LogoThrusterColor
     Scale Factor=&MakeVector4_FromLinearColor(LinearColor=<rgba 5 5 5 1>)
    )
    Lifetime=&UniformRangedFloat(Minimum=0.12, Maximum=0.5)
    Module.Write Color=true
    Module.Write Lifetime=true
    Module.Write Mass=true
    Module.Write Material Random=true
    Module.Write Position=false
    Module.Write Scale=false
    Module.Write SpriteRotation=false
    Module.Write SpriteSize=true
    Position=&SimulationPosition()
    Sprite Size=&UniformRangedVector2D(Minimum=<1 1>, Maximum=<1.2 2>)
   ) {enabled=false}
   /Niagara/Modules/Spawn/Initialization/V2/InitializeParticle(
    Color=&MultiplyLinearColor(
     Linear Color=$System.LogoThrusterColor
     Scale Factor=&MakeVector4_FromLinearColor(LinearColor=<rgba 5 5 5 1>)
    )
    Lifetime=&UniformRangedFloat(Minimum=0.12, Maximum=0.5)
    Sprite Size=&UniformRangedVector2D(Minimum=<1 1>, Maximum=<1.2 2>)
    Sprite Size Mode=::NewEnumerator4
    UsePositionOffset=false
    Write Material Random=true
   ) {module=InitializeParticle001}
   &SkeletalMeshLocation(
    Bone / Socket Index=$Particles.PartitionParticles.Partition
    Bone Sampling Mode=::NewEnumerator4
    Inherit Velocity=false
    Interpolate Mesh Sampling=false
    MaskSpawnGroup=false
    Sampled Position Offset Coordinate Space:ENiagaraMeshTransforms=::NewEnumerator3
    Skeletal Mesh=$System.Skeletal Mesh
   )
   /Niagara/Modules/Spawn/Location/SphereLocation(
    MaskSpawnGroup=false
    Module.Override Randomness=false
    Sphere Radius=&UniformRangedFloat(Minimum=5, Maximum=15)
    Surface Only=false
   ) {enabled=false}
   &AddVelocity() {enabled=false, module=AddVelocity001}
   &AddVelocity(Velocity=&UniformRangedVector(Minimum=<-100 -100 -150>, Maximum=<100 100 -80>)) {enabled=false}
   &ShapeLocation(MaskSpawnGroup=false, Override Seed=false, Sphere Radius=5, UseOwnerScale=true)
   &AddVelocity(
    Cone Angle=90
    Cone Axis=<0 0 -1>
    Module.UseConeFalloff=false
    Module.UseVelDistribution=false
    Override Seed=false
    Velocity Mode=::NewEnumerator2
    Velocity Speed=&RandomRangeFloat(Minimum=100, Maximum=350)
   ) {module=AddVelocity002}
  ]
  particleUpdateScript [
   &ParticleState()
   &AccelerationForce(Acceleration=<0 0 -200>)
   &SkeletalMeshLocation(
    Apply To Particle Position=false
    Bone / Socket Index=$Particles.PartitionParticles.Partition
    Bone Sampling Mode=::NewEnumerator4
    Inherit Velocity=false
    Interpolate Mesh Sampling=false
    MaskSpawnGroup=false
    Sampled Position Offset Coordinate Space:ENiagaraMeshTransforms=::NewEnumerator3
    Skeletal Mesh=$System.Skeletal Mesh
   ) {module=SkeletalMeshLocation001}
   /Niagara/VectorFields/SampleVectorField(
    Sampled Vector Scale=0.4
    VectorField=DataInterface() {propertyValues {Field=/Game/Effects/VectorFields/Noisy_Turbulence_103, bTileX=true, bTileY=true, bTileZ=true}}
   ) {enabled=false}
   /Niagara/VectorFields/ApplyVectorField(Vector Field Intensity as Force=0, Vector Field Intensity as Velocity=50, Vector Field Velocity Tightness=0.15) {enabled=false}
   &Drag(Drag=3, Use Linear Drag=true, Use Rotational Drag=true)
   &InheritSourceMovement(
    Applied Position Scale=&VectorFromFloat(Value=&FloatFromCurve(
      FloatCurve=DataInterface() {propertyValues {Curve {keys [
          {tangentMode=RCTM_User, time=0, value=1}
          {arriveTangent=-8.0089, interpMode=RCIM_Cubic, leaveTangent=-8.0089, tangentMode=RCTM_User, time=1, value=0}
         ]}}}
      Scale Curve=0.1
     ))
    ApplyForce=false
    ApplyPosition=true
    ApplyPositionFalloff=false
    Source Velocity:Vector3f=$Output.SkeletalMeshLocation001.SampledVelocity
   )
   &SolveForcesAndVelocity(Clamp Velocity=false, Limit Acceleration=false)
   &ScaleColor(
    Scale Alpha=&FloatFromCurve(FloatCurve=DataInterface() {propertyValues {Curve {keys [<at 0 1>,<at 1 0>]}}})
    Scale RGB=&MakeVectorFromLinearColorRGB(Color=&ColorFromCurve(ColorCurve=DataInterface() {propertyValues {
        AlphaCurve {keys []}
        BlueCurve {keys [<at 0 0.25 RCIM_Cubic>,<at 0.5 0.5>,<at 1 0.875 RCIM_Cubic>]}
        GreenCurve {keys [<at 0 0.525 RCIM_Cubic>,<at 0.5 0.6833>,<at 1 0.9208 RCIM_Cubic>]}
        RedCurve {keys [<at 0 1 RCIM_Cubic>,<at 0.5 1>,<at 1 1 RCIM_Cubic>]}
       }}))
    ScaleA=true
    ScaleRGB=true
   )
   &ScaleColor(
    Scale RGB=&VectorFromFloat(Value=&FloatFromCurve(
      FloatCurve=DataInterface() {propertyValues {Curve {keys [<at 0 20 RCIM_Cubic>, {arriveTangent=-20, interpMode=RCIM_Cubic, leaveTangent=-20, time=0.25, value=5}, <at 1 0 RCIM_Cubic>]}}}
      Scale Curve=0.2
     ))
    ScaleA=true
    ScaleRGB=true
   ) {module=ScaleColor001}
   &ScaleSpriteSizeBySpeed(Max Scale Factor=<1 8>, Min Scale Factor=<1 0.5>)
   &ScaleSpriteSize(Scale Factor=&Vector2DFromFloat(Value=&FloatFromCurve(FloatCurve=DataInterface() {propertyValues {Curve {keys [
          {arriveTangent=43.4604, interpMode=RCIM_Cubic, leaveTangent=43.4604, tangentMode=RCTM_User, time=0.0001, value=0.0578}
          {arriveTangent=-0.0578, interpMode=RCIM_Cubic, leaveTangent=-0.0578, time=0.03, value=1}
          {arriveTangent=-3.2182, interpMode=RCIM_Cubic, leaveTangent=-3.2182, tangentMode=RCTM_User, time=1, value=0}
         ]}}})))
  ]
  rendererClasses [&NiagaraSpriteRendererProperties]
  renderers [{rendererClass=&NiagaraSpriteRendererProperties}]
 }
 Sparks {
  simTarget=GPUComputeSim
  emitterUpdateScript [
   &EmitterState(
    Life Cycle Mode=::NewEnumerator1
    Loop Behavior=::NewEnumerator1
    Loop Delay=$System.FirstThrusterDelay
    Loop Duration=&Subtract_Float(A=$System.MainThrusterStart, B=$System.FirstThrusterDelay)
    UseLoopDelay=true
   )
   &SpawnRate(SpawnRate=350, Use Spawn Probability=false) {module=SpawnRate001}
   &SpawnRate(SpawnRate=200, Use Spawn Probability=false) {enabled=false}
  ]
  particleSpawnScript [
   &PartitionParticles()
   /Niagara/Modules/Spawn/Initialization/InitializeParticle(
    Color=&MultiplyLinearColor(
     Linear Color=$System.LogoThrusterColor
     Scale Factor=&MakeVector4_FromLinearColor(LinearColor=<rgba 5 5 5 1>)
    )
    Lifetime=&RandomRangeFloat(Minimum=0.1, Maximum=0.3)
    Module.Write Color=true
    Module.Write Lifetime=true
    Module.Write Mass=true
    Module.Write Material Random=true
    Module.Write Position=false
    Module.Write Scale=false
    Module.Write SpriteRotation=false
    Module.Write SpriteSize=true
    Position=&SimulationPosition()
    Sprite Size=&UniformRangedVector2D(Minimum=<1 2>, Maximum=<0.5 4>)
   )
   /SetVariables(Particles.SourceSelect:NiagaraInt32=&MakeCustomIntFromBool(
     Bool=&SetBoolByFloatComparison(A=$System.Age, B=$System.SecondThrusterDelay)
     True INT=&RandomRangeInt(Minimum=0, Maximum=3)
     False INT=&Add_Integer(A=$Particles.PartitionParticles.Partition, B=0)
    )) {isSetParameters=true}
   &FN_SetSpawnGroup(Number of Spawn Groups=5) {enabled=false}
   &SkeletalMeshLocation(
    Apply Sampled Orientation to Mesh Particles=true
    Bone / Socket Index=&SpawnGroup()
    Bone Sampling Mode=::NewEnumerator4
    Inherit Velocity=false
    Interpolate Mesh Sampling=false
    MaskSpawnGroup=false
    Sampled Position Offset=<46 0 0>
    Sampled Position Offset Coordinate Space:ENiagaraMeshTransforms=::NewEnumerator3
    Skeletal Mesh=$User.Skeletal Mesh
    Use Sampled Mesh Attribute as Ribbon ID=false
   ) {enabled=false}
   &SkeletalMeshLocation(
    Apply Sampled Orientation to Mesh Particles=true
    Bone / Socket Index=$Particles.SourceSelect
    Bone Sampling Mode=::NewEnumerator4
    Inherit Velocity=false
    Interpolate Mesh Sampling=false
    MaskSpawnGroup=false
    Sampled Position Offset=&MergeHelper_VectorSwitch(0=<-10 0 0>, 1=<10 0 0>, 2=<0 -10 0>, 3=<0 10 0>, 4=<0 0 0>, Index=$Particles.SourceSelect, Num=::NewEnumerator3)
    Sampled Position Offset Coordinate Space:ENiagaraMeshTransforms=::NewEnumerator3
    Skeletal Mesh=$System.Skeletal Mesh
   ) {module=SkeletalMeshLocation001}
   &AddVelocityInCone(
    Cone Axis=&RotateVectorByQuaternion(
     Vector To Rotate=<1 0 0>
     Quaternion=&MultiplyQuaternion(
      Quaternion A=$Output.SkeletalMeshLocation001.SampledRotation
      Quaternion B=&MergeHelper_QuatSwitch(
       0:Quat4f=&MakeQuatFromAxisAngle(
        Axis=<0 1 0>
        AngleInDegrees=&UniformAOrBFloat(A=-60, B=60, Evaluation Type=::NewEnumerator0, Recalculate Random Each Loop=true, Override Seed=false)
       )
       1:Quat4f=&MakeQuatFromAxisAngle(
        Axis=<0 1 0>
        AngleInDegrees=&UniformAOrBFloat(A=-120, B=110, Evaluation Type=::NewEnumerator0, Recalculate Random Each Loop=true, Override Seed=false)
       )
       2:Quat4f=&MakeQuatFromAxisAngle(Axis=<0 0 1>, AngleInDegrees=-90)
       3:Quat4f=&MakeQuatFromAxisAngle(Axis=<0 0 1>, AngleInDegrees=90)
       4:Quat4f=?
       5:Quat4f=?
       6:Quat4f=?
       7:Quat4f=?
       8:Quat4f=?
       9:Quat4f=?
       Index:NiagaraInt32=$Particles.SourceSelect
       Num:MergeHelper_SwitcherEnum=::NewEnumerator2
      )
     )
    )
    Module.Use Velocity Falloff On Cone Axis=false
    Velocity Strength=&RandomRangeFloat(Minimum=500, Maximum=750)
   )
  ]
  particleUpdateScript [
   &Drag(Drag=8, Use Linear Drag=true, Use Rotational Drag=true)
   &SolveForcesAndVelocity(Clamp Velocity=false, Limit Acceleration=false)
   &ParticleState()
   &ScaleSpriteSizeBySpeed(Max Scale Factor=<0.5 2>, Velocity Threshold=1250)
   &ScaleColor(
    Scale Alpha=&FloatFromCurve(
     FloatCurve=DataInterface() {propertyValues {Curve {keys [<at 0 1>,<at 1 0>]}}}
     CurveIndex=&Multiply_Float(A=$Particles.NormalizedAge, B=2)
    )
    Scale RGB=&MakeVectorFromLinearColorRGB(Color=&ColorFromCurve(
      ColorCurve=DataInterface() {propertyValues {
        AlphaCurve {keys []}
        BlueCurve {keys [<at 0 0.25 RCIM_Cubic>,<at 0.5 0.5>,<at 1 0.875 RCIM_Cubic>]}
        GreenCurve {keys [<at 0 0.525 RCIM_Cubic>,<at 0.5 0.6833>,<at 1 0.9208 RCIM_Cubic>]}
        RedCurve {keys [<at 0 1 RCIM_Cubic>,<at 0.5 1>,<at 1 1 RCIM_Cubic>]}
       }}
      CurveIndex=&Multiply_Float(A=$Particles.NormalizedAge, B=2)
     ))
    ScaleA=true
    ScaleRGB=true
   )
   &ScaleSpriteSize(Scale Factor=&Vector2DFromCurve(
     Vector2Curve=DataInterface() {propertyValues {
       XCurve {keys [<at 0 1>,<at 1 0>]}
       YCurve {keys [{arriveTangent=18.2053, interpMode=RCIM_Cubic, leaveTangent=18.2053, tangentMode=RCTM_User, time=0, value=1}, <at 1 10>]}
      }}
     CurveIndex=&Multiply_Float(A=$Particles.NormalizedAge, B=2)
     ScaleCurve=<1 1>
    )) {enabled=false}
  ]
  rendererClasses [&NiagaraSpriteRendererProperties]
  renderers [{rendererClass=&NiagaraSpriteRendererProperties}]
 }
}
summary {
 systemName=NS_Emote_TruckScale_Pawn
 emitters [
  {emitterName=Sparks, simTarget=GPUComputeSim, rendererClasses [&NiagaraSpriteRendererProperties]}
  {emitterName=Flame_Launch, rendererClasses [&NiagaraMeshRendererProperties, &NiagaraSpriteRendererProperties]}
  {emitterName=Calf_Flames, rendererClasses [&NiagaraMeshRendererProperties, &NiagaraSpriteRendererProperties]}
  {emitterName=Smoke, rendererClasses [&NiagaraSpriteRendererProperties]}
  {emitterName=Rim_Glow, rendererClasses [&NiagaraSpriteRendererProperties, &NiagaraLightRendererProperties]}
  {emitterName=FootThrusters, rendererClasses [&NiagaraMeshRendererProperties, &NiagaraSpriteRendererProperties]}
  {
   emitterName=Animorphic_-_HotCoreStart_-_FlameOuterStart
   rendererClasses [&NiagaraSpriteRendererProperties, &NiagaraMeshRendererProperties]
  }
  {emitterName=Spark_Start, simTarget=GPUComputeSim, rendererClasses [&NiagaraSpriteRendererProperties]}
 ]
 userVariables [
  User.ForwardDot {type=NiagaraFloat, defaultValue=0, description=""}
  User.Skeletal Mesh {
   type=NiagaraDataInterfaceSkeletalMesh
   defaultValue {
    struct=NiagaraExt_VariableValue_DataInterface
    value {dataInterface=&self:NiagaraDataInterfaceSkeletalMesh_1}
   }
  }
 ]
}
